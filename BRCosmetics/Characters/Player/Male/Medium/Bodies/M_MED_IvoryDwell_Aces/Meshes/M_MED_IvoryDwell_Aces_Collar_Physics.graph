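# Physics asset for the standing jacket collar of the M_MED_IvoryDwell_Aces male player body cosmetic.
# Seven collar segments (center, three left, three right) each pair a small capsule base body with a thin box dynamic body.
# Each pair's constraint locks twist and one swing axis, leaving the other a narrow limit so segments act as hinged flaps.
# Head and upper-arm twist capsules supply body collision shapes around the collar.
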PhysicsAsset /BRCosmetics/Characters/Player/Male/Medium/Bodies/M_MED_IvoryDwell_Aces/Meshes/M_MED_IvoryDwell_Aces_Collar_Physics
bodies {
 base_jacketCollarStand_1_c [{shapeType=Capsule, length=0.25, radius=0.5, shapeName=base_jacketCollarStand_3_l_capsule}]
 base_jacketCollarStand_1_l [{shapeType=Capsule, length=0.25, radius=0.5, shapeName=base_jacketCollarStand_3_l_capsule}]
 base_jacketCollarStand_1_r [{shapeType=Capsule, length=0.25, radius=0.5, shapeName=base_jacketCollarStand_3_l_capsule}]
 base_jacketCollarStand_2_l [{shapeType=Capsule, length=0.25, radius=0.5, shapeName=base_jacketCollarStand_3_l_capsule}]
 base_jacketCollarStand_2_r [{shapeType=Capsule, length=0.25, radius=0.5, shapeName=base_jacketCollarStand_3_l_capsule}]
 base_jacketCollarStand_3_l [{shapeType=Capsule, length=0.25, radius=0.5, shapeName=base_jacketCollarStand_3_l_capsule}]
 base_jacketCollarStand_3_r [{shapeType=Capsule, length=0.25, radius=0.5, shapeName=base_jacketCollarStand_3_l_capsule}]
 coll_upperarm_twist_01_l [{shapeType=Capsule, length=15.3157, radius=6.0414, shapeName=upperarm_twist_01_l_capsule, center=<11 0.5 -1.8>, rotation=<rot -90 0 0>}]
 coll_upperarm_twist_01_r [{
   shapeType=Capsule
   length=15.3157
   radius=6.0414
   shapeName=upperarm_twist_01_r_capsule
   center=<-11 2.2529 2.2363>
   rotation=<rot -90 180 0>
  }]
 dyn_jacketCollarStand_1_c [{shapeType=Box, extentX=7.5, extentY=12.1773, extentZ=0.7827, shapeName=dyn_jacketCollarStand_1_c_box, center=<4.7258 0 0>}]
 dyn_jacketCollarStand_1_l [{
   shapeType=Box
   extentX=9.3317
   extentY=8
   extentZ=0.7827
   shapeName=dyn_jacketCollarStand_1_l_box
   center=<5.6 4.05 -0.524>
   rotation=<rot 0 0 5>
  }]
 dyn_jacketCollarStand_1_r [{shapeType=Box, extentX=9.3317, extentY=8, extentZ=0.7827, shapeName=dyn_jacketCollarStand_1_l_box, center=<5.6 -3.8 -0.5>}]
 dyn_jacketCollarStand_2_l [{
   shapeType=Box
   extentX=7.5
   extentY=12.1773
   extentZ=0.7827
   shapeName=dyn_jacketCollarStand_2_l_box
   center=<3.9 0 0>
   rotation=<rot 0 10 10>
  }]
 dyn_jacketCollarStand_2_r [{
   shapeType=Box
   extentX=7.5
   extentY=12.1773
   extentZ=0.7827
   shapeName=dyn_jacketCollarStand_2_r_box
   center=<3.9 0 0>
   rotation=<rot 0 -10 -10>
  }]
 dyn_jacketCollarStand_3_l [{shapeType=Box, extentX=7.5, extentY=12.1773, extentZ=0.7827, shapeName=dyn_jacketCollarStand_3_l_box, center=<4.7258 0 -0.5122>}]
 dyn_jacketCollarStand_3_r [{shapeType=Box, extentX=7.5, extentY=12.1773, extentZ=0.7827, shapeName=dyn_jacketCollarStand_3_r_box, center=<4.7258 0 0.5122>}]
 head [
  {shapeType=Capsule, length=9.6392, radius=9.5473, shapeName=head_capsule, center=<4.9632 -0.2772 0>, rotation=<rot 90 23.7934 0>}
  {shapeType=Capsule, length=10.175, radius=4.9626, shapeName=head_capsule, center=<-1.1424 6.7796 0>}
 ]
}
constraints [
 {bone1Name=dyn_jacketCollarStand_1_c, bone2Name=base_jacketCollarStand_1_c, swing1LimitDegrees=0, swing2LimitDegrees=20, twistLimitDegrees=0}
 {bone1Name=dyn_jacketCollarStand_1_l, bone2Name=base_jacketCollarStand_1_l, swing1LimitDegrees=0, swing2LimitDegrees=20, twistLimitDegrees=0}
 {bone1Name=dyn_jacketCollarStand_1_r, bone2Name=base_jacketCollarStand_1_r, swing1LimitDegrees=0, swing2LimitDegrees=20, twistLimitDegrees=0}
 {bone1Name=dyn_jacketCollarStand_2_l, bone2Name=base_jacketCollarStand_2_l, swing1LimitDegrees=0, swing2LimitDegrees=20, twistLimitDegrees=0}
 {bone1Name=dyn_jacketCollarStand_2_r, bone2Name=base_jacketCollarStand_2_r, swing1LimitDegrees=0, swing2LimitDegrees=20, twistLimitDegrees=0}
 {bone1Name=dyn_jacketCollarStand_3_l, bone2Name=base_jacketCollarStand_3_l, swing1LimitDegrees=0, swing2LimitDegrees=20, twistLimitDegrees=0}
 {bone1Name=dyn_jacketCollarStand_3_r, bone2Name=base_jacketCollarStand_3_r, swing1LimitDegrees=0, swing2LimitDegrees=20, twistLimitDegrees=0}
]
