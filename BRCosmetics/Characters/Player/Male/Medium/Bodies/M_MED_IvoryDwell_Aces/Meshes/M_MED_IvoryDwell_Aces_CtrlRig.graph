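# Control Rig for the IvoryDwell Aces male character body that drives corrective deformation of jacket collar, hood and shoulders.
# The solve runs FortDeformRig, a shared lower-body function, then collar setup, measuring against nulls created during construction.
# Collar bones follow weighted translation-only parent constraints, then receive offsets weighted by swing-twist angle remaps and spherical pose readers on head, neck, clavicles and right upper arm.
# One right-hood correction also requires pistol use and interpolated targeting state.
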ControlRigBlueprint /BRCosmetics/Characters/Player/Male/Medium/Bodies/M_MED_IvoryDwell_Aces/Meshes/M_MED_IvoryDwell_Aces_CtrlRig
graphs [
 CollapseNode_ContainedGraph {nodes [
   n1 = RigVMFunctionEntryNode()
   RigVMFunction_Sequence = RigVMAggregateNode(ExecuteContext=@n1.ExecutePin)
   &RigUnit_ParentConstraint(
    AdvancedSettings=(RotationOrderForFilter=XZY)
    Child=(Type=Bone,Name="base_jacketCollarStand_1_l")
    Filter=(TranslationFilter=(bX=True,bY=True,bZ=True),RotationFilter=(bX=False,bY=False,bZ=False),ScaleFilter=(bX=False,bY=False,bZ=False))
    Parents="((Item=(Type=Bone,Name=\"spine_02\"),Weight=0.0164),(Item=(Type=Bone,Name=\"neck_02\"),Weight=0.0186),(Item=(Type=Bone,Name=\"deform_pec_l\"),Weight=0.0392),(Item=(Type=Bone,Name=\"clavicle_l\"),Weight=0.0528),(Item=(Type=Bone,Name=\"spine_03\"),Weight=0.0572),(Item ...(truncated)"
    Weight=1
    ExecutePin=@RigVMFunction_Sequence.A
   )
   > &RigUnit_ParentConstraint(
    AdvancedSettings=(RotationOrderForFilter=XZY)
    Child=(Type=Bone,Name="base_jacketCollarStand_2_l")
    Filter=(TranslationFilter=(bX=True,bY=True,bZ=True),RotationFilter=(bX=False,bY=False,bZ=False),ScaleFilter=(bX=False,bY=False,bZ=False))
    Parents=((Item=(Type=Bone,Name="spine_04"),Weight=0.0128),(Item=(Type=Bone,Name="neck_02"),Weight=0.0568),(Item=(Type=Bone,Name="clavicle_l"),Weight=0.1229),(Item=(Type=Bone,Name="spine_05"),Weight=0.8075))
    Weight=1
   )
   > &RigUnit_ParentConstraint(
    AdvancedSettings=(RotationOrderForFilter=XZY)
    Child=(Type=Bone,Name="base_jacketCollarStand_3_l")
    Filter=(TranslationFilter=(bX=True,bY=True,bZ=True),RotationFilter=(bX=False,bY=False,bZ=False),ScaleFilter=(bX=False,bY=False,bZ=False))
    Parents=((Item=(Type=Bone,Name="neck_01"),Weight=0.0135),(Item=(Type=Bone,Name="kin_hood_def_2"),Weight=0.0311),(Item=(Type=Bone,Name="neck_02"),Weight=0.0616),(Item=(Type=Bone,Name="clavicle_l"),Weight=0.134),(Item=(Type=Bone,Name="spine_05"),Weight=0.7599))
    Weight=1
   )
   > &RigUnit_ParentConstraint(
    AdvancedSettings=(RotationOrderForFilter=XZY)
    Child=(Type=Bone,Name="base_jacketCollarStand_1_c")
    Filter=(TranslationFilter=(bX=True,bY=True,bZ=True),RotationFilter=(bX=False,bY=False,bZ=False),ScaleFilter=(bX=False,bY=False,bZ=False))
    Parents=((Item=(Type=Bone,Name="neck_01"),Weight=0.0157),(Item=(Type=Bone,Name="neck_02"),Weight=0.0609),(Item=(Type=Bone,Name="kin_hood_def_2"),Weight=0.1578),(Item=(Type=Bone,Name="spine_05"),Weight=0.7656))
    Weight=1
   )
   > &RigUnit_ParentConstraint(
    AdvancedSettings=(RotationOrderForFilter=XZY)
    Child=(Type=Bone,Name="base_jacketCollarStand_1_r")
    Filter=(TranslationFilter=(bX=True,bY=True,bZ=True),RotationFilter=(bX=False,bY=False,bZ=False),ScaleFilter=(bX=False,bY=False,bZ=False))
    Parents="((Item=(Type=Bone,Name=\"neck_02\"),Weight=0.0101),(Item=(Type=Bone,Name=\"spine_02\"),Weight=0.0163),(Item=(Type=Bone,Name=\"deform_pec_r\"),Weight=0.0529),(Item=(Type=Bone,Name=\"spine_03\"),Weight=0.0568),(Item=(Type=Bone,Name=\"clavicle_r\"),Weight=0.0786),(Item ...(truncated)"
    Weight=1
   )
   > &RigUnit_ParentConstraint(
    AdvancedSettings=(RotationOrderForFilter=XZY)
    Child=(Type=Bone,Name="base_jacketCollarStand_2_r")
    Filter=(TranslationFilter=(bX=True,bY=True,bZ=True),RotationFilter=(bX=False,bY=False,bZ=False),ScaleFilter=(bX=False,bY=False,bZ=False))
    Parents=((Item=(Type=Bone,Name="spine_04"),Weight=0.0125),(Item=(Type=Bone,Name="neck_02"),Weight=0.0488),(Item=(Type=Bone,Name="clavicle_r"),Weight=0.1588),(Item=(Type=Bone,Name="spine_05"),Weight=0.7798))
    Weight=1
   )
   > &RigUnit_ParentConstraint(
    AdvancedSettings=(RotationOrderForFilter=XZY)
    Child=(Type=Bone,Name="base_jacketCollarStand_3_r")
    Filter=(TranslationFilter=(bX=True,bY=True,bZ=True),RotationFilter=(bX=False,bY=False,bZ=False),ScaleFilter=(bX=False,bY=False,bZ=False))
    Parents=((Item=(Type=Bone,Name="neck_01"),Weight=0.0142),(Item=(Type=Bone,Name="neck_02"),Weight=0.072),(Item=(Type=Bone,Name="clavicle_r"),Weight=0.132),(Item=(Type=Bone,Name="spine_05"),Weight=0.7818))
    Weight=1
   )
   GetAlphaFromQuatDelta = RigVMFunctionReferenceNode(
    Rotation Order=ZYX
    child=(Type=Bone,Name="head")
    onlyTwist=true
    outputMinMax=(X=0,Y=1)
    parent=(Type=Null,Name="head_sp_null")
    premult=1
    rotationMinMax=(X=0,Y=-30)
    twistAxis=(X=1,Y=0,Z=0)
    ExecuteContext=@RigVMFunction_Sequence.B
   )
   > &RigUnit_ModifyTransforms(
    ItemToModify="((Item=(Type=Bone,Name=\"kin_hood_def_1_l\"),Transform=(Rotation=(X=0,Y=0,Z=0,W=1),Translation=(X=-1.65,Y=0,Z=0),Scale3D=(X=1,Y=1,Z=1))),(Item=(Type=Bone,Name=\"kin_hoodSide_l\"),Transform=(Rotation=(X=0.1029,Y=0.1727,Z=-0.0182,W=0.9794),Translation=(X=0,Y=0,Z ...(truncated)"
    WeightMinimum=0
    Weight=@GetAlphaFromQuatDelta.alpha
   )
   > GetAlphaFromQuatDelta_1 = RigVMFunctionReferenceNode(
    Rotation Order=ZYX
    child=(Type=Bone,Name="head")
    onlyTwist=true
    outputMinMax=(X=0,Y=1)
    parent=(Type=Null,Name="head_sp_null")
    premult=1
    rotationMinMax=(X=0,Y=30)
    twistAxis=(X=1,Y=0,Z=0)
   )
   > &RigUnit_ModifyTransforms(
    ItemToModify="((Item=(Type=Bone,Name=\"kin_hood_def_1_r\"),Transform=(Rotation=(X=0,Y=0,Z=0,W=1),Translation=(X=1.65,Y=0,Z=0),Scale3D=(X=1,Y=1,Z=1))),(Item=(Type=Bone,Name=\"kin_hoodSide_r\"),Transform=(Rotation=(X=-0.1029,Y=0.1727,Z=0.0182,W=0.9794),Translation=(X=0,Y=0,Z= ...(truncated)"
    WeightMinimum=0
    Weight=@GetAlphaFromQuatDelta_1.alpha
   )
   > GetAlphaFromQuatDelta_2 = RigVMFunctionReferenceNode(
    Rotation Order=ZYX
    child=(Type=Bone,Name="neck_01")
    debugRotation=false
    onlyTwist=false
    outputMinMax=(X=0,Y=1)
    parent=(Type=Null,Name="neck1_sp_null")
    premult=1
    rotationMinMax=(X=0,Y=-40)
    triggerAxis=Z
    twistAxis=(X=1,Y=0,Z=0)
   )
   > GetAlphaFromQuatDelta_5 = RigVMFunctionReferenceNode(
    Rotation Order=ZYX
    child=(Type=Bone,Name="head")
    debugRotation=false
    onlyTwist=false
    outputMinMax=(X=0,Y=1)
    parent=(Type=Null,Name="head_sp_null")
    premult=1
    rotationMinMax=(X=-15,Y=-60)
    triggerAxis=Z
    twistAxis=(X=1,Y=0,Z=0)
   )
   n2 = &RigVMFunction_MathDoubleMax(A=@GetAlphaFromQuatDelta_5.alpha, B=@GetAlphaFromQuatDelta_2.alpha)
   > &RigUnit_ModifyTransforms(
    ItemToModify=((Item=(Type=Bone,Name="kin_hood_def_1_c"),Transform=(Rotation=(X=0.1352,Y=0,Z=0,W=0.9908),Translation=(X=0,Y=3,Z=-2),Scale3D=(X=1,Y=1,Z=1))))
    WeightMinimum=0
    Weight=@n2.Result
   )
   > &RigUnit_ModifyTransforms(
    ItemToModify="((Item=(Type=Bone,Name=\"base_jacketCollarStand_1_c\"),Transform=(Rotation=(X=0,Y=0.4574,Z=0,W=-0.8893),Translation=(X=-0.75,Y=0,Z=1.3),Scale3D=(X=1,Y=1,Z=1))),(Item=(Type=Bone,Name=\"base_jacketCollarStand_3_l\"),Transform=(Rotation=(X=0,Y=-0.2164,Z=0,W=0.976 ...(truncated)"
    WeightMinimum=0
    Weight=@n2.Result
   )
   > GetAlphaFromQuatDelta_8 = RigVMFunctionReferenceNode(
    Rotation Order=ZYX
    child=(Type=Bone,Name="head")
    onlyTwist=false
    outputMinMax=(X=0,Y=1)
    parent=(Type=Null,Name="head_sp_null")
    premult=1
    rotationMinMax=(X=0,Y=-25)
    triggerAxis=Y
    twistAxis=(X=1,Y=0,Z=0)
   )
   > &RigUnit_ModifyTransforms(
    ItemToModify=((Item=(Type=Bone,Name="base_jacketCollarStand_3_r"),Transform=(Rotation=(X=0,Y=-0.1305,Z=0,W=0.9914),Translation=(X=0,Y=0,Z=0),Scale3D=(X=1,Y=1,Z=1))))
    WeightMinimum=0
    Weight=@GetAlphaFromQuatDelta_8.alpha
   )
   > GetAlphaFromQuatDelta_9 = RigVMFunctionReferenceNode(
    Rotation Order=ZYX
    child=(Type=Bone,Name="head")
    onlyTwist=false
    outputMinMax=(X=0,Y=1)
    parent=(Type=Null,Name="head_sp_null")
    premult=1
    rotationMinMax=(X=0,Y=25)
    triggerAxis=Y
    twistAxis=(X=1,Y=0,Z=0)
   )
   > &RigUnit_ModifyTransforms(
    ItemToModify=((Item=(Type=Bone,Name="base_jacketCollarStand_3_l"),Transform=(Rotation=(X=0,Y=-0.1564,Z=0,W=0.9877),Translation=(X=0,Y=0,Z=0),Scale3D=(X=1,Y=1,Z=1))))
    WeightMinimum=0
    Weight=@GetAlphaFromQuatDelta_9.alpha
   )
   GetAlphaFromQuatDelta_3 = RigVMFunctionReferenceNode(
    Rotation Order=ZYX
    child=(Type=Bone,Name="clavicle_l")
    debugRotation=false
    outputMinMax=(X=0,Y=1)
    parent=(Type=Null,Name="clavicle_l_null")
    premult=1
    rotationMinMax=(X=-20,Y=-45)
    triggerAxis=Y
    twistAxis=(X=1,Y=0,Z=0)
    ExecuteContext=@RigVMFunction_Sequence.C
   )
   > &RigUnit_ModifyTransforms(
    ItemToModify="((Item=(Type=Bone,Name=\"base_jacketCollarStand_2_l\"),Transform=(Rotation=(X=-0.173,Y=0.0858,Z=0.0151,W=0.9811),Translation=(X=0.65,Y=0,Z=1),Scale3D=(X=1,Y=1,Z=1))),(Item=(Type=Bone,Name=\"base_jacketCollarStand_1_l\"),Transform=(Rotation=(X=0,Y=0.1305,Z=0,W= ...(truncated)"
    WeightMinimum=0
    Weight=@GetAlphaFromQuatDelta_3.alpha
   )
   > &RigUnit_ModifyTransforms(
    ItemToModify="((Item=(Type=Bone,Name=\"kin_hoodSide_l\"),Transform=(Rotation=(X=0.2132,Y=0.1695,Z=-0.0376,W=0.9615),Translation=(X=0,Y=0,Z=0),Scale3D=(X=1,Y=1,Z=1))),(Item=(Type=Bone,Name=\"kin_hood_def_1_l\"),Transform=(Rotation=(X=0,Y=0,Z=0,W=1),Translation=(X=-5,Y=0,Z=0) ...(truncated)"
    WeightMinimum=0
    Weight=@GetAlphaFromQuatDelta_3.alpha
   )
   > GetAlphaFromQuatDelta_6 = RigVMFunctionReferenceNode(
    Rotation Order=ZYX
    child=(Type=Bone,Name="clavicle_l")
    debugRotation=false
    outputMinMax=(X=0,Y=1)
    parent=(Type=Null,Name="clavicle_l_null")
    premult=1
    rotationMinMax=(X=-35,Y=-45)
    triggerAxis=Y
    twistAxis=(X=1,Y=0,Z=0)
   )
   > &RigUnit_ModifyTransforms(
    ItemToModify=((Item=(Type=Bone,Name="deform_upperarm_up_l"),Transform=(Rotation=(X=0,Y=0,Z=0,W=1),Translation=(X=3,Y=0,Z=-3),Scale3D=(X=1,Y=1,Z=1))))
    WeightMinimum=0
    Weight=@GetAlphaFromQuatDelta_6.alpha
   )
   > n3 = &RigUnit_SphericalPoseReader(
    ActiveRegionScaleFactors=(PositiveWidth=1,NegativeWidth=1,PositiveHeight=2,NegativeHeight=1)
    ActiveRegionSize=0.1
    Debug=(bDrawDebug=false,bDraw2D=False,bDrawLocalAxes=False,DebugScale=25,DebugSegments=20,DebugThickness=0.25)
    DriverAxis=(X=1,Y=0,Z=0)
    DriverItem=(Type=Bone,Name="clavicle_l")
    FalloffRegionScaleFactors=(PositiveWidth=1,NegativeWidth=1,PositiveHeight=2,NegativeHeight=1)
    FalloffSize=0.25
    FlipHeightScaling=False
    FlipWidthScaling=False
    OptionalParentItem=(Type=Bone,Name="None")
    RotationOffset=(X=24,Y=-40,Z=0)
   )
   > &RigUnit_ModifyTransforms(
    ItemToModify=((Item=(Type=Bone,Name="base_jacketCollarStand_2_l"),Transform=(Rotation=(X=0.1802,Y=0.0891,Z=-0.0949,W=0.975),Translation=(X=0,Y=0,Z=-2),Scale3D=(X=1,Y=1,Z=1))))
    WeightMinimum=0
    Weight=@n3.OutputParam
   )
   > n4 = &RigUnit_SphericalPoseReader(
    ActiveRegionScaleFactors=(PositiveWidth=1.5,NegativeWidth=1,PositiveHeight=1.5,NegativeHeight=1.5)
    ActiveRegionSize=0.1
    Debug=(bDrawDebug=false,bDraw2D=False,bDrawLocalAxes=False,DebugScale=25,DebugSegments=20,DebugThickness=0.25)
    DriverAxis=(X=-1,Y=0,Z=0)
    DriverItem=(Type=Bone,Name="upperarm_r")
    FalloffRegionScaleFactors=(PositiveWidth=1.5,NegativeWidth=1,PositiveHeight=1,NegativeHeight=1)
    FalloffSize=0.2
    FlipHeightScaling=False
    FlipWidthScaling=False
    OptionalParentItem=(Type=Bone,Name="None")
    RotationOffset=(X=-93.7294,Y=25.4067,Z=35.8226)
   )
   > &RigUnit_ModifyTransforms(
    ItemToModify=((Item=(Type=Bone,Name="base_jacketCollarStand_1_r"),Transform=(Rotation=(X=0,Y=0.3827,Z=0,W=0.9239),Translation=(X=0,Y=0,Z=0),Scale3D=(X=1,Y=1,Z=1))))
    WeightMinimum=0
    Weight=@n4.OutputParam
   )
   > n5 = &RigUnit_SphericalPoseReader(
    ActiveRegionScaleFactors=(PositiveWidth=1.8,NegativeWidth=1,PositiveHeight=1.8,NegativeHeight=1.8)
    ActiveRegionSize=0.08
    Debug=(bDrawDebug=false,bDraw2D=False,bDrawLocalAxes=False,DebugScale=25,DebugSegments=20,DebugThickness=0.25)
    DriverAxis=(X=-1,Y=0,Z=0)
    DriverItem=(Type=Bone,Name="upperarm_r")
    FalloffRegionScaleFactors=(PositiveWidth=1,NegativeWidth=1,PositiveHeight=1,NegativeHeight=1)
    FalloffSize=0.2
    FlipHeightScaling=False
    FlipWidthScaling=False
    OptionalParentItem=(Type=Bone,Name="None")
    RotationOffset=(X=-110,Y=0,Z=20)
   )
   n6 = &RigVMFunction_AlphaInterp(
    InterpSpeedDecreasing=10
    InterpSpeedIncreasing=10
    Scale=1
    bClampResult=False
    Value=&RigVMFunction_MathBoolToFloat(Value=RigVMVariableNode() {variable=IsTargeting}.Value).Result
   )
   Cast = &RigVMFunction_MathBoolToFloat(Value=RigVMVariableNode() {variable=isUsingPistol}.Value)
   Multiply_2 = RigVMAggregateNode(C=@n6.Result, B=@Cast.Result, A=@n5.OutputParam)
   > &RigUnit_ModifyTransforms(
    ItemToModify="((Item=(Type=Bone,Name=\"kin_hoodSide_r\"),Transform=(Rotation=(X=0,Y=0.1045,Z=0,W=0.9945),Translation=(X=0,Y=0,Z=0),Scale3D=(X=1,Y=1,Z=1))),(Item=(Type=Bone,Name=\"kin_hood_def_2_r\"),Transform=(Rotation=(X=0,Y=0,Z=0,W=-1),Translation=(X=3,Y=-5,Z=-2.5),Scale3 ...(truncated)"
    WeightMinimum=0
    Weight=@Multiply_2.Result
   )
   > n7 = &RigUnit_SphericalPoseReader(
    ActiveRegionScaleFactors=(PositiveWidth=1,NegativeWidth=1,PositiveHeight=0.5,NegativeHeight=1)
    ActiveRegionSize=0.15
    Debug=(bDrawDebug=false,bDraw2D=False,bDrawLocalAxes=False,DebugScale=25,DebugSegments=20,DebugThickness=0.25)
    DriverAxis=(X=1,Y=0,Z=0)
    DriverItem=(Type=Bone,Name="head")
    FalloffRegionScaleFactors=(PositiveWidth=0.8,NegativeWidth=0.8,PositiveHeight=0.8,NegativeHeight=1)
    FalloffSize=0.25
    FlipHeightScaling=False
    FlipWidthScaling=False
    OptionalParentItem=(Type=Bone,Name="spine_05")
    RotationOffset=(X=50,Y=-50,Z=-100)
   )
   > &RigUnit_ModifyTransforms(
    ItemToModify="((Item=(Type=Bone,Name=\"base_jacketCollarStand_3_r\"),Transform=(Rotation=(X=0,Y=-0.1305,Z=0,W=0.9914),Translation=(X=0,Y=0,Z=1),Scale3D=(X=1,Y=1,Z=1))),(Item=(Type=Bone,Name=\"kin_hood_def_1_r\"),Transform=(Rotation=(X=0,Y=0,Z=0,W=1),Translation=(X=3,Y=0,Z=0 ...(truncated)"
    WeightMinimum=0
    Weight=@n7.OutputParam
   )
   > n8 = &RigUnit_SphericalPoseReader(
    ActiveRegionScaleFactors=(PositiveWidth=1,NegativeWidth=1,PositiveHeight=0.5,NegativeHeight=1)
    ActiveRegionSize=0.15
    Debug=(bDrawDebug=false,bDraw2D=False,bDrawLocalAxes=False,DebugScale=25,DebugSegments=20,DebugThickness=0.25)
    DriverAxis=(X=1,Y=0,Z=0)
    DriverItem=(Type=Bone,Name="head")
    FalloffRegionScaleFactors=(PositiveWidth=0.8,NegativeWidth=0.8,PositiveHeight=0.8,NegativeHeight=1)
    FalloffSize=0.25
    FlipHeightScaling=False
    FlipWidthScaling=False
    OptionalParentItem=(Type=Bone,Name="spine_05")
    RotationOffset=(X=-150,Y=-50,Z=100)
   )
   > &RigUnit_ModifyTransforms(
    ItemToModify="((Item=(Type=Bone,Name=\"base_jacketCollarStand_3_l\"),Transform=(Rotation=(X=0,Y=-0.1305,Z=0,W=0.9914),Translation=(X=0,Y=0,Z=1),Scale3D=(X=1,Y=1,Z=1))),(Item=(Type=Bone,Name=\"kin_hood_def_1_l\"),Transform=(Rotation=(X=0,Y=0,Z=0,W=1),Translation=(X=-3,Y=0,Z= ...(truncated)"
    WeightMinimum=0
    Weight=@n8.OutputParam
   )
   > n9 = &RigUnit_SphericalPoseReader(
    ActiveRegionScaleFactors=(PositiveWidth=2,NegativeWidth=2,PositiveHeight=2,NegativeHeight=1)
    ActiveRegionSize=0.05
    Debug=(bDrawDebug=false,bDraw2D=False,bDrawLocalAxes=False,DebugScale=25,DebugSegments=20,DebugThickness=0.25)
    DriverAxis=(X=-1,Y=0,Z=0)
    DriverItem=(Type=Bone,Name="clavicle_r")
    FalloffRegionScaleFactors=(PositiveWidth=1,NegativeWidth=1,PositiveHeight=1,NegativeHeight=1)
    FalloffSize=0.2
    FlipHeightScaling=False
    FlipWidthScaling=False
    OptionalParentItem=(Type=Bone,Name="None")
    RotationOffset=(X=140,Y=0,Z=100)
   )
   > &RigUnit_ModifyTransforms(
    ItemToModify="((Item=(Type=Bone,Name=\"deform_upperarm_up_r\"),Transform=(Rotation=(X=0,Y=0.1736,Z=0,W=0.9848),Translation=(X=0,Y=6,Z=3),Scale3D=(X=1,Y=1,Z=1))),(Item=(Type=Bone,Name=\"kin_hood_def_2_r\"),Transform=(Rotation=(X=0,Y=0,Z=0,W=1),Translation=(X=0,Y=0,Z=3),Scale ...(truncated)"
    WeightMinimum=0
    Weight=@n9.OutputParam
   )
   > n10 = &RigUnit_SphericalPoseReader(
    ActiveRegionScaleFactors=(PositiveWidth=2,NegativeWidth=2,PositiveHeight=2,NegativeHeight=1)
    ActiveRegionSize=0.05
    Debug=(bDrawDebug=false,bDraw2D=False,bDrawLocalAxes=False,DebugScale=25,DebugSegments=20,DebugThickness=0.25)
    DriverAxis=(X=1,Y=0,Z=0)
    DriverItem=(Type=Bone,Name="clavicle_l")
    FalloffRegionScaleFactors=(PositiveWidth=1,NegativeWidth=1,PositiveHeight=1,NegativeHeight=1)
    FalloffSize=0.2
    FlipHeightScaling=False
    FlipWidthScaling=False
    OptionalParentItem=(Type=Bone,Name="None")
    RotationOffset=(X=-40,Y=0,Z=100)
   )
   > &RigUnit_ModifyTransforms(
    ItemToModify="((Item=(Type=Bone,Name=\"deform_upperarm_up_l\"),Transform=(Rotation=(X=0,Y=0.1736,Z=0,W=0.9848),Translation=(X=0,Y=-6,Z=-3),Scale3D=(X=1,Y=1,Z=1))),(Item=(Type=Bone,Name=\"kin_hood_def_2_l\"),Transform=(Rotation=(X=0,Y=0,Z=0,W=1),Translation=(X=0,Y=0,Z=-3),Sc ...(truncated)"
    WeightMinimum=0
    Weight=@n10.OutputParam
   )
   GetAlphaFromQuatDelta_4 = RigVMFunctionReferenceNode(
    Rotation Order=ZYX
    child=(Type=Bone,Name="clavicle_r")
    debugRotation=false
    outputMinMax=(X=0,Y=1)
    parent=(Type=Null,Name="clavicle_r_null")
    premult=1
    rotationMinMax=(X=-20,Y=-45)
    triggerAxis=Y
    twistAxis=(X=1,Y=0,Z=0)
    ExecuteContext=@RigVMFunction_Sequence.D
   )
   > &RigUnit_ModifyTransforms(
    ItemToModify="((Item=(Type=Bone,Name=\"base_jacketCollarStand_2_r\"),Transform=(Rotation=(X=0.173,Y=0.0858,Z=-0.0151,W=0.9811),Translation=(X=0.65,Y=0,Z=1),Scale3D=(X=1,Y=1,Z=1))),(Item=(Type=Bone,Name=\"base_jacketCollarStand_1_r\"),Transform=(Rotation=(X=0,Y=0.1305,Z=0,W= ...(truncated)"
    WeightMinimum=0
    Weight=@GetAlphaFromQuatDelta_4.alpha
   )
   > &RigUnit_ModifyTransforms(
    ItemToModify="((Item=(Type=Bone,Name=\"kin_hoodSide_r\"),Transform=(Rotation=(X=-0.2132,Y=0.1695,Z=0.0376,W=0.9615),Translation=(X=0,Y=0,Z=0),Scale3D=(X=1,Y=1,Z=1))),(Item=(Type=Bone,Name=\"kin_hood_def_1_r\"),Transform=(Rotation=(X=0,Y=0,Z=0,W=1),Translation=(X=5,Y=0,Z=0), ...(truncated)"
    WeightMinimum=0
    Weight=@GetAlphaFromQuatDelta_4.alpha
   )
   > GetAlphaFromQuatDelta_7 = RigVMFunctionReferenceNode(
    Rotation Order=ZYX
    child=(Type=Bone,Name="clavicle_r")
    debugRotation=false
    outputMinMax=(X=0,Y=1)
    parent=(Type=Null,Name="clavicle_r_null")
    premult=1
    rotationMinMax=(X=-35,Y=-45)
    triggerAxis=Y
    twistAxis=(X=1,Y=0,Z=0)
   )
   > &RigUnit_ModifyTransforms(
    ItemToModify=((Item=(Type=Bone,Name="deform_upperarm_up_r"),Transform=(Rotation=(X=0,Y=0,Z=0,W=1),Translation=(X=-3,Y=0,Z=3),Scale3D=(X=1,Y=1,Z=1))))
    WeightMinimum=0
    Weight=@GetAlphaFromQuatDelta_7.alpha
   )
   > n11 = &RigUnit_SphericalPoseReader(
    ActiveRegionScaleFactors=(PositiveWidth=1,NegativeWidth=1,PositiveHeight=2,NegativeHeight=1)
    ActiveRegionSize=0.1
    Debug=(bDrawDebug=false,bDraw2D=False,bDrawLocalAxes=False,DebugScale=25,DebugSegments=20,DebugThickness=0.25)
    DriverAxis=(X=-1,Y=0,Z=0)
    DriverItem=(Type=Bone,Name="clavicle_r")
    FalloffRegionScaleFactors=(PositiveWidth=1,NegativeWidth=1,PositiveHeight=2,NegativeHeight=1)
    FalloffSize=0.25
    FlipHeightScaling=False
    FlipWidthScaling=False
    OptionalParentItem=(Type=Bone,Name="None")
    RotationOffset=(X=-156,Y=-40,Z=0)
   )
   > &RigUnit_ModifyTransforms(
    ItemToModify=((Item=(Type=Bone,Name="base_jacketCollarStand_2_r"),Transform=(Rotation=(X=-0.1802,Y=0.0891,Z=0.0949,W=0.975),Translation=(X=0,Y=0,Z=-2),Scale3D=(X=1,Y=1,Z=1))))
    WeightMinimum=0
    Weight=@n11.OutputParam
   )
   RigVMFunctionReturnNode(ExecutePin=@RigVMFunction_Sequence.E)
  ]}
 Multiply_3_ContainedGraph {nodes [
   n1 = RigVMFunctionEntryNode()
   RigVMFunctionReturnNode(Result=&RigVMFunction_MathDoubleMul(B=@n1.C, A=&RigVMFunction_MathDoubleMul(A=@n1.A, B=@n1.B).Result).Result)
  ]}
 Multiply_3_ContainedGraph {nodes [
   n1 = RigVMFunctionEntryNode()
   RigVMFunctionReturnNode(Result=&RigVMFunction_MathTransformMul(B=@n1.C, A=&RigVMFunction_MathTransformMul(A=@n1.A, B=@n1.B) {template="Multiply::Execute(in A,in B,out Result)"}.Result) {template="Multiply::Execute(in A,in B,out Result)"}.Result)
  ]}
 New Function_ContainedGraph {nodes [
   RigVMFunctionEntryNode()
   RigVMFunction_MathBoolToInteger = &RigVMFunction_MathBoolToInteger(Value=RigVMVariableNode() {variable=onlyTwist}.Value)
   RigVMFunction_MathQuaternionSwingTwist = &RigVMFunction_MathQuaternionSwingTwist(
    Input=&RigUnit_GetRelativeTransformForItem(
     bChildInitial=False
     bParentInitial=False
     Child=RigVMVariableNode() {variable=child}.Value
     Parent=RigVMVariableNode() {variable=parent}.Value
    ).RelativeTransform.Rotation
    TwistAxis=RigVMVariableNode() {variable=twistAxis}.Value
   )
   n1 = RigVMDispatchNode(
    Values=((),())
    Index=@RigVMFunction_MathBoolToInteger.Result
    Values.0=@RigVMFunction_MathQuaternionSwingTwist.Swing
    Values.1=@RigVMFunction_MathQuaternionSwingTwist.Twist
   ) {template="DISPATCH_RigVMDispatch_SelectInt32(in Index,in Values,out Result)"}
   > RigVMVariableNode(Value=@n1.Result) {variable=entry_outputQuat}
   RigVMFunction_MathQuaternionToEuler = &RigVMFunction_MathQuaternionToEuler(Value=@n1.Result, RotationOrder=RigVMVariableNode() {variable=Rotation Order}.Value)
   > DISPATCH_RigVMDispatch_Print = RigVMDispatchNode(
    ScreenColor=(R=1,G=1,B=1,A=1)
    ScreenDuration=1.5
    Value=@RigVMFunction_MathQuaternionToEuler.Result
    Enabled=RigVMVariableNode() {variable=debugRotation}.Value
   ) {template="DISPATCH_RigVMDispatch_Print(in Prefix,in Value,in Enabled,in ScreenDuration,in ScreenColor)"}
   Cast = RigVMDispatchNode(Value=RigVMVariableNode() {variable=triggerAxis}.Value) {template="DISPATCH_RigVMDispatch_CastEnumToInt(in Value,out Result)"}
   DISPATCH_RigVMDispatch_SelectInt32 = RigVMDispatchNode(
    Values=(0,0,0)
    Index=@Cast.Result
    Values.0=@RigVMFunction_MathQuaternionToEuler.Result.X
    Values.1=@RigVMFunction_MathQuaternionToEuler.Result.Y
    Values.2=@RigVMFunction_MathQuaternionToEuler.Result.Z
   ) {template="DISPATCH_RigVMDispatch_SelectInt32(in Index,in Values,out Result)"}
   n2 = RigVMVariableNode() {variable=rotationMinMax}
   n3 = RigVMVariableNode() {variable=outputMinMax}
   Remap = &RigVMFunction_MathDoubleRemap(
    bClamp=False
    Value=@DISPATCH_RigVMDispatch_SelectInt32.Result
    SourceMinimum=@n2.Value.X
    SourceMaximum=@n2.Value.Y
    TargetMinimum=@n3.Value.X
    TargetMaximum=@n3.Value.Y
   )
   Absolute = &RigVMFunction_MathDoubleAbs(Value=@Remap.Result)
   n4 = RigVMVariableNode() {variable=absoluteValue}
   Multiply = &RigVMFunction_MathDoubleMul(
    A=&RigVMFunction_MathDoubleClamp(Value=RigVMDispatchNode(True=@Absolute.Result, False=@Remap.Result, Condition=@n4.Value) {template="DISPATCH_RigVMDispatch_If(in Condition,in True,in False,out Result)"}.Result).Result
    B=RigVMVariableNode() {variable=premult}.Value
   )
   > RigVMFunctionReturnNode(alpha=@Multiply.Result, rawQuat=RigVMVariableNode() {variable=entry_outputQuat}.Value)
  ]}
 New Function_ContainedGraph {nodes [
   RigVMFunctionEntryNode()
   Num = RigVMDispatchNode(Array=RigVMVariableNode() {variable=targets}.Value) {template="DISPATCH_RigVMDispatch_ArrayGetNum(in Array,out Num)"}
   > RigVMDispatchNode(Element=0, Num=@Num.Num, Array=RigVMVariableNode() {variable=outputValues}.Value) {template="DISPATCH_RigVMDispatch_ArrayInit(io Array,in Num,in Element)"}
   > RigVMDispatchNode(Element=0, Array=RigVMVariableNode() {variable=divisors}.Value, Num=@Num.Num) {template="DISPATCH_RigVMDispatch_ArrayInit(io Array,in Num,in Element)"}
   > RigVMDispatchNode(Element=0, Array=RigVMVariableNode() {variable=outLenRemaped}.Value, Num=@Num.Num) {template="DISPATCH_RigVMDispatch_ArrayInit(io Array,in Num,in Element)"}
   n1 = RigVMVariableNode() {variable=child}
   > RigVMVariableNode(Value=&RigVMFunction_MathTransformMul(
     A=(Rotation=(X=0,Y=0,Z=0,W=1),Translation=(X=0,Y=0,Z=0),Scale3D=(X=1,Y=1,Z=1))
     B=&RigUnit_GetTransform(bInitial=False, Item=@n1.Value).Transform
     A.Translation=RigVMVariableNode() {variable=primaryAxis}.Value
    ) {template="Multiply::Execute(in A,in B,out Result)"}.Result.Translation) {variable=target vector}
   n2 = RigVMVariableNode() {variable=parent}
   > RigVMVariableNode(Value=&RigVMFunction_MathTransformMul(
     A=&RigUnit_GetRelativeTransformForItem(bChildInitial=true, bParentInitial=true, Child=@n1.Value, Parent=@n2.Value).RelativeTransform
     B=&RigUnit_GetTransform(bInitial=False, Item=@n2.Value).Transform
    ) {template="Multiply::Execute(in A,in B,out Result)"}.Result) {variable=parentChildLocated transform}
   > /Script/RigVM.RigVMFunction_DebugLineNoSpace(
    Color=(R=0,G=0,B=1,A=1)
    DebugDrawSettings=(DepthPriority=SDPG_Foreground,Lifetime=-1)
    Thickness=0.5
    WorldOffset=(Rotation=(X=0,Y=0,Z=0,W=1),Translation=(X=0,Y=0,Z=0),Scale3D=(X=1,Y=1,Z=1))
    B=RigVMVariableNode() {variable=target vector}.Value
    A=RigVMVariableNode() {variable=parentChildLocated transform}.Value.Translation
    bEnabled=RigVMVariableNode() {variable=drawLines}.Value
   ) {is=struct RigVMUnitNode}
   > For_Each = RigVMDispatchNode(Array=RigVMVariableNode() {variable=targets}.Value) {template="DISPATCH_RigVMDispatch_ArrayIterator(in Array,out Element,out Index,out Count,out Ratio)"}
   > RigVMVariableNode(Value=@For_Each.Index) {variable=index1 iter}
   Multiply_1 = RigVMAggregateNode(
    A=(Rotation=(X=0,Y=0,Z=0,W=1),Translation=(X=0,Y=0,Z=0),Scale3D=(X=1,Y=1,Z=1))
    B=(Rotation=(X=0,Y=0,Z=0,W=1),Translation=(X=0,Y=0,Z=0),Scale3D=(X=1,Y=1,Z=1))
    C=RigVMVariableNode() {variable=parentChildLocated transform}.Value
    A.Translation=RigVMVariableNode() {variable=primaryAxis}.Value
    B.Rotation=@For_Each.Element
   )
   > RigVMFunction_DebugTransformMutableNoSpace = /Script/RigVM.RigVMFunction_DebugTransformMutableNoSpace(
    Color=(R=1,G=0,B=1,A=0)
    DebugDrawSettings=(DepthPriority=SDPG_Foreground,Lifetime=-1)
    Mode=Box
    Scale=1
    Thickness=0.5
    Transform=(Rotation=(X=0,Y=0,Z=0,W=1),Translation=(X=0,Y=0,Z=0),Scale3D=(X=1,Y=1,Z=1))
    WorldOffset=(Rotation=(X=0,Y=0,Z=0,W=1),Translation=(X=0,Y=0,Z=0),Scale3D=(X=1,Y=1,Z=1))
    bEnabled=RigVMVariableNode() {variable=drawLines}.Value
    Transform.Translation=@Multiply_1.Result.Translation
   ) {is=struct RigVMUnitNode}
   n3 = RigVMVariableNode() {variable=parentChildLocated transform}
   Unit = /Script/RigVM.RigVMFunction_MathVectorUnit(Value=/Script/RigVM.RigVMFunction_MathVectorSub(A=RigVMVariableNode() {variable=target vector}.Value, B=@n3.Value.Translation) {is=struct RigVMUnitNode, template="Subtract::Execute(in A,in B,out Result)"}.Result) {is=struct RigVMUnitNode, template="Unit::Execute(in Value,out Result)"}
   Subtract = &RigVMFunction_MathVectorSub(
    A=@Unit.Result
    B=&RigVMFunction_MathVectorUnit(Value=&RigVMFunction_MathVectorSub(A=@Multiply_1.Result.Translation, B=@n3.Value.Translation) {template="Subtract::Execute(in A,in B,out Result)"}.Result) {template="Unit::Execute(in Value,out Result)"}.Result
   ) {template="Subtract::Execute(in A,in B,out Result)"}
   RigVMFunction_MathVectorLength = /Script/RigVM.RigVMFunction_MathVectorLength(Value=@Subtract.Result) {is=struct RigVMUnitNode, template="Length::Execute(in Value,out Result)"}
   Remap = &RigVMFunction_MathDoubleRemap(
    SourceMaximum=0
    TargetMaximum=1
    TargetMinimum=0
    bClamp=False
    SourceMinimum=RigVMVariableNode() {variable=remapMult}.Value
    Value=@RigVMFunction_MathVectorLength.Result
   )
   > Set_At_2 = RigVMDispatchNode(Element=@Remap.Result, Array=RigVMVariableNode() {variable=outLenRemaped}.Value, Index=RigVMVariableNode() {variable=index1 iter}.Value) {template="DISPATCH_RigVMDispatch_ArraySetAtIndex(io Array,in Index,in Element)"}
   Power = /Script/RigVM.RigVMFunction_MathDoublePow(B=2, A=@RigVMFunction_MathVectorLength.Result) {is=struct RigVMUnitNode, template="Power::Execute(in A,in B,out Result)"}
   Clamp = &RigVMFunction_MathDoubleClamp(Maximum=10000000000, Value=@Power.Result)
   Divide = &RigVMFunction_MathDoubleDiv(A=1, B=@Clamp.Result)
   > Set_At = RigVMDispatchNode(Element=@Divide.Result, Index=RigVMVariableNode() {variable=index1 iter}.Value, Array=RigVMVariableNode() {variable=divisors}.Value) {template="DISPATCH_RigVMDispatch_ArraySetAtIndex(io Array,in Index,in Element)"}
   n4 = RigVMVariableNode() {variable=divisors}
   For_Each_1 = RigVMDispatchNode(ExecuteContext=@For_Each.Completed, Array=@n4.Value) {template="DISPATCH_RigVMDispatch_ArrayIterator(in Array,out Element,out Index,out Count,out Ratio)"}
   ArraySum = /Script/RigVM.RigVMFunction_MathDoubleArraySum(Array=@n4.Value) {is=struct RigVMUnitNode, template="ArraySum::Execute(in Array,out Sum)"}
   n5 = &RigVMFunction_MathDoubleDiv(B=@ArraySum.Sum, A=@For_Each_1.Element)
   Multiply = &RigVMFunction_MathDoubleMul(
    B=RigVMDispatchNode(Index=@For_Each_1.Index, Array=RigVMVariableNode() {variable=outLenRemaped}.Value) {template="DISPATCH_RigVMDispatch_ArrayGetAtIndex(in Array,in Index,out Element)"}.Element
    A=@n5.Result
   )
   RigVMFunction_AnimEvalRichCurve = &RigVMFunction_AnimEvalRichCurve(
    SourceMinimum=0
    TargetMinimum=0
    Value=RigVMDispatchNode(False=@n5.Result, True=@Multiply.Result, Condition=RigVMVariableNode() {variable=ease}.Value) {template="DISPATCH_RigVMDispatch_If(in Condition,in True,in False,out Result)"}.Result
    Curve=RigVMVariableNode() {variable=remapCurve}.Value
   )
   > Set_At_1 = RigVMDispatchNode(Index=@For_Each_1.Index, Element=@RigVMFunction_AnimEvalRichCurve.Result, Array=RigVMVariableNode() {variable=outputValues}.Value) {template="DISPATCH_RigVMDispatch_ArraySetAtIndex(io Array,in Index,in Element)"}
   RigVMFunctionReturnNode(weights=RigVMVariableNode() {variable=outputValues}.Value, ExecuteContext=@For_Each_1.Completed)
  ]}
 RigVMFunctionLibrary {nodes [
   GetAlphaFromQuatDelta = RigVMCollapseNode(
    ExecuteContext=()
    Rotation Order=ZYX
    child=(Type=Bone,Name="None")
    outputMinMax=(X=0,Y=1)
    parent=(Type=Bone,Name="None")
    premult=1
    rotationMinMax=(X=0,Y=0)
    twistAxis=(X=1,Y=0,Z=0)
   )
   InverseDistance = RigVMCollapseNode(
    ExecuteContext=()
    child=(Type=Bone,Name="None")
    drawLines=true
    parent=(Type=Bone,Name="None")
    primaryAxis=(X=0,Y=0,Z=0)
    remapCurve=(EditorCurveData=(Keys=((),(Time=0.5,Value=0.5,ArriveTangent=1.203,LeaveTangent=1.203),(Time=1,Value=1))))
    remapMult=4
    targets=()
   )
  ]}
 RigVMFunction_Sequence_1_ContainedGraph {nodes [
   n1 = RigVMFunctionEntryNode()
   RigVMFunction_Sequence = &RigVMFunction_Sequence(ExecuteContext=@n1.ExecuteContext)
   n2 = &RigVMFunction_Sequence(ExecuteContext=@RigVMFunction_Sequence.B)
   n3 = &RigVMFunction_Sequence(ExecuteContext=@n2.B)
   n4 = &RigVMFunction_Sequence(ExecuteContext=@n3.B)
   RigVMFunctionReturnNode(A=@RigVMFunction_Sequence.A, B=@n2.A, C=@n3.A, D=@n4.A, E=@n4.B)
  ]}
 RigVMFunction_Sequence_1_ContainedGraph {nodes [
   n1 = RigVMFunctionEntryNode()
   RigVMFunction_Sequence = &RigVMFunction_Sequence(ExecuteContext=@n1.ExecuteContext)
   n2 = &RigVMFunction_Sequence(ExecuteContext=@RigVMFunction_Sequence.B)
   n3 = &RigVMFunction_Sequence(ExecuteContext=@n2.B)
   RigVMFunctionReturnNode(A=@RigVMFunction_Sequence.A, B=@n2.A, C=@n3.A, D=@n3.B)
  ]}
 RigVMModel {nodes [
   n1 = &RigUnit_BeginExecution()
   /Script/FortniteGame.RigUnit_FortDeformRig(
    Bool=(bUpperArm=true,bElbow=true,bWrist=true,bGroin=false,bGlute=false,bKnee=true,bPec=true,bBicep=False,bLat=False)
    Curve="(bLUpperArmUpAlpha=False,bRUpperArmUpAlpha=False,bLUpperArmFrAlpha=False,bRUpperArmFrAlpha=False,bLUpperArmBkAlpha=False,bRUpperArmBkAlpha=False,bLWristUpAlpha=False,bRWristUpAlpha=False,bLWristDnAlpha=False,bRWristDnAlpha=False,bLKneeAlpha=False,bRKneeAlp ...(truncated)"
    Debug=(bUpperArmDebug=False,bWristDebug=False,bLatDebug=False,bGroinGluteDebug=False)
    Value="(UpperArmUp=(Rotation=(X=0.0,Y=-0.0,Z=0.0,W=1.0),Translation=(X=0.0,Y=0.0,Z=3),Scale3D=(X=0.5,Y=1.0,Z=1.0)),UpperArmFwdFr=(Rotation=(X=0.0,Y=-0.0,Z=0.0,W=1.0),Translation=(X=0.0,Y=10.0,Z=0.0),Scale3D=(X=0.50,Y=1.0,Z=1.0)),UpperArmFwdBk=(Rotation=(X=0.0,Y=- ...(truncated)"
    ExecutePin=@n1.ExecutePin
   ) {is=struct RigVMUnitNode}
   > IvoryDwellChar_lowerBody = RigVMFunctionReferenceNode()
   jacketCollarSetup = RigVMCollapseNode(ExecutePin=@IvoryDwellChar_lowerBody.ExecuteContext)
  ]}
 RigVMModel Rig {nodes [
   n1 = &RigUnit_PrepareForExecution()
   RigVMFunction_Sequence = RigVMAggregateNode(ExecuteContext=@n1.ExecutePin)
   internalNametarget_hoodieCollar_1_c_null = &RigUnit_HierarchyAddNull(
    Name=target_hoodieCollar_1_c_null
    Parent=(Type=Bone,Name="spine_05")
    Space=LocalSpace
    Transform=(Rotation=(X=0,Y=0,Z=0,W=1),Translation=(X=22.0973,Y=-11.7982,Z=0),Scale3D=(X=1,Y=1,Z=1))
    ExecutePin=@RigVMFunction_Sequence.B
   )
   > internalNametarget_hoodieCollar_3_l_null = &RigUnit_HierarchyAddNull(
    Name=target_hoodieCollar_3_l_null
    Parent=(Type=Bone,Name="spine_05")
    Space=LocalSpace
    Transform=(Rotation=(X=0,Y=0,Z=0,W=1),Translation=(X=11.4447,Y=8.885,Z=-5.8925),Scale3D=(X=1,Y=1,Z=1))
   )
   > internalNametarget_hoodieCollar_2_l_null = &RigUnit_HierarchyAddNull(
    Name=target_hoodieCollar_2_l_null
    Parent=(Type=Bone,Name="spine_05")
    Space=LocalSpace
    Transform=(Rotation=(X=0,Y=0,Z=0,W=1),Translation=(X=17.3515,Y=1.9876,Z=-11.4143),Scale3D=(X=1,Y=1,Z=1))
   )
   > internalNametarget_hoodieCollar_1_l_null = &RigUnit_HierarchyAddNull(
    Name=target_hoodieCollar_1_l_null
    Parent=(Type=Bone,Name="spine_05")
    Space=LocalSpace
    Transform=(Rotation=(X=0,Y=0,Z=0,W=1),Translation=(X=20.4135,Y=-7.0318,Z=-8.2927),Scale3D=(X=1,Y=1,Z=1))
   )
   > internalNametarget_hoodieCollar_3_r_null = &RigUnit_HierarchyAddNull(
    Name=target_hoodieCollar_3_r_null
    Parent=(Type=Bone,Name="spine_05")
    Space=LocalSpace
    Transform=(Rotation=(X=0,Y=0,Z=0,W=1),Translation=(X=11.4447,Y=8.8848,Z=5.8924),Scale3D=(X=1,Y=1,Z=1))
   )
   > internalNametarget_hoodieCollar_2_r_null = &RigUnit_HierarchyAddNull(
    Name=target_hoodieCollar_2_r_null
    Parent=(Type=Bone,Name="spine_05")
    Space=LocalSpace
    Transform=(Rotation=(X=0,Y=0,Z=0,W=1),Translation=(X=17.3517,Y=1.9877,Z=11.4146),Scale3D=(X=1,Y=1,Z=1))
   )
   > internalNametarget_hoodieCollar_1_r_null = &RigUnit_HierarchyAddNull(
    Name=target_hoodieCollar_1_r_null
    Parent=(Type=Bone,Name="spine_05")
    Space=LocalSpace
    Transform=(Rotation=(X=0,Y=0,Z=0,W=1),Translation=(X=20.4136,Y=-7.0318,Z=8.2927),Scale3D=(X=1,Y=1,Z=1))
   )
   > internalNamehead_sp_null = &RigUnit_HierarchyAddNull(
    Name=head_sp_null
    Parent=(Type=Bone,Name="spine_05")
    Space=LocalSpace
    Transform=(Rotation=(X=0,Y=0,Z=0.036,W=0.9994),Translation=(X=20.6573,Y=3.5489,Z=0),Scale3D=(X=1,Y=1,Z=1))
   )
   > &RigUnit_HierarchyAddNull(
    Name=neck1_sp_null
    Parent=(Type=Bone,Name="spine_05")
    Space=LocalSpace
    Transform=&RigUnit_GetRelativeTransformForItem(Child=(Type=Bone,Name="neck_01"), Parent=(Type=Bone,Name="spine_05"), bChildInitial=true, bParentInitial=true).RelativeTransform
   )
   internalNamehipChains_twist_bone_1 = /Script/ControlRig.RigUnit_HierarchyAddBone(
    Name=hipChains_twist_bone
    Parent=(Type=Bone,Name="pelvis")
    Space=LocalSpace
    Transform=(Rotation=(X=0.0075,Y=-0.975,Z=0.1399,W=0.1726),Translation=(X=-3.9069,Y=3.4534,Z=12.7954),Scale3D=(X=1,Y=1,Z=1))
    ExecutePin=@RigVMFunction_Sequence.C
   ) {is=struct RigVMUnitNode}
   > internalNamehipChains_target_bone_1 = &RigUnit_HierarchyAddBone(
    Name=hipChains_target_bone
    Parent=(Type=Bone,Name="pelvis")
    Space=LocalSpace
    Transform=(Rotation=(X=0.0075,Y=-0.975,Z=0.1399,W=0.1726),Translation=(X=-19.5746,Y=4.015,Z=18.4371),Scale3D=(X=1,Y=1,Z=1))
   )
   > internalNamebase_hemJacket_1_l_null = &RigUnit_HierarchyAddNull(
    Name=base_hemJacket_1_l_null
    Parent=(Type=Bone,Name="kin_hemJacket")
    Space=LocalSpace
    Transform=(Rotation=(X=0.0283,Y=-0.2471,Z=-0.9685,W=0.0068),Translation=(X=1.1224,Y=13.0099,Z=-9.0976),Scale3D=(X=1,Y=1,Z=1))
   )
   > internalNamebase_hemJacket_1_r_null = &RigUnit_HierarchyAddNull(
    Name=base_hemJacket_1_r_null
    Parent=(Type=Bone,Name="kin_hemJacket")
    Space=LocalSpace
    Transform=(Rotation=(X=-0.0283,Y=0.2471,Z=-0.9685,W=0.0068),Translation=(X=1.1219,Y=13.0098,Z=9.0976),Scale3D=(X=1,Y=1,Z=1))
   )
   internalNametarget_hemJacket_1_r_null = &RigUnit_HierarchyAddNull(
    Name=target_hemJacket_1_r_null
    Parent=(Type=Bone,Name="spine_02")
    Space=LocalSpace
    Transform=(Rotation=(X=0,Y=0,Z=0,W=1),Translation=(X=-14.2961,Y=12.5915,Z=9.8928),Scale3D=(X=1,Y=1,Z=1))
    ExecutePin=@RigVMFunction_Sequence.D
   )
   > internalNametarget_hemJacket_1_l_null = &RigUnit_HierarchyAddNull(
    Name=target_hemJacket_1_l_null
    Parent=(Type=Bone,Name="spine_02")
    Space=LocalSpace
    Transform=(Rotation=(X=0,Y=0,Z=0,W=1),Translation=(X=-14.2961,Y=12.5915,Z=-9.8928),Scale3D=(X=1,Y=1,Z=1))
   )
   > internalNametarget_hemJacket_2_r_null = &RigUnit_HierarchyAddNull(
    Name=target_hemJacket_2_r_null
    Parent=(Type=Bone,Name="spine_02")
    Space=LocalSpace
    Transform=(Rotation=(X=0,Y=0,Z=0,W=1),Translation=(X=-13.4146,Y=4.6931,Z=18.7961),Scale3D=(X=1,Y=1,Z=1))
   )
   > internalNametarget_hemJacket_2_l_null = &RigUnit_HierarchyAddNull(
    Name=target_hemJacket_2_l_null
    Parent=(Type=Bone,Name="spine_02")
    Space=LocalSpace
    Transform=(Rotation=(X=0,Y=0,Z=0,W=1),Translation=(X=-13.4149,Y=4.6931,Z=-18.7961),Scale3D=(X=1,Y=1,Z=1))
   )
   > internalNametarget_hemJacket_3_r_null = &RigUnit_HierarchyAddNull(
    Name=target_hemJacket_3_r_null
    Parent=(Type=Bone,Name="spine_02")
    Space=LocalSpace
    Transform=(Rotation=(X=0,Y=0,Z=0,W=1),Translation=(X=-12.5935,Y=-7.6634,Z=17.0597),Scale3D=(X=1,Y=1,Z=1))
   )
   > internalNametarget_hemJacket_3_l_null = &RigUnit_HierarchyAddNull(
    Name=target_hemJacket_3_l_null
    Parent=(Type=Bone,Name="spine_02")
    Space=LocalSpace
    Transform=(Rotation=(X=0,Y=0,Z=0,W=1),Translation=(X=-12.5935,Y=-7.6634,Z=-17.0597),Scale3D=(X=1,Y=1,Z=1))
   )
   > internalNametarget_jacketHem_back_null = &RigUnit_HierarchyAddNull(
    Name=target_jacketHem_back_null
    Parent=(Type=Bone,Name="spine_02")
    Space=LocalSpace
    Transform=(Rotation=(X=0,Y=0,Z=0,W=1),Translation=(X=-12.6545,Y=-13.5338,Z=-0.4968),Scale3D=(X=1,Y=1,Z=1))
   )
   ItemArray = RigVMDispatchNode(Value="((Type=Bone,Name=\"calf_l\"),(Type=Bone,Name=\"calf_r\"),(Type=Bone,Name=\"thigh_l\"),(Type=Bone,Name=\"thigh_r\"),(Type=Bone,Name=\"foot_l\"),(Type=Bone,Name=\"foot_r\"),(Type=Bone,Name=\"clavicle_l\"),(Type=Bone,Name=\"clavicle_r\"),(Type=Bone,Name=\"hand_l\"),(Type=Bone, ...(truncated)") {template="DISPATCH_RigVMDispatch_Constant(io Value)"}
   DISPATCH_RigVMDispatch_ArrayIterator = RigVMDispatchNode(Array=@ItemArray.Value, ExecuteContext=@RigVMFunction_Sequence.A) {template="DISPATCH_RigVMDispatch_ArrayIterator(in Array,out Element,out Index,out Count,out Ratio)"}
   Concat = /Script/RigVM.RigVMFunction_NameConcat(B=_null, A=@DISPATCH_RigVMDispatch_ArrayIterator.Element.Name) {is=struct RigVMUnitNode, template="Concat::Execute(in A,in B,out Result)"}
   > &RigUnit_HierarchyAddNull(
    Space=LocalSpace
    Name=@Concat.Result
    Transform=&RigUnit_GetTransform(Item=(Type=Bone,Name="clavicle_l"), Space=LocalSpace, bInitial=true, Item.Name=@DISPATCH_RigVMDispatch_ArrayIterator.Element.Name).Transform
    Parent=/Script/ControlRig.RigUnit_HierarchyGetParent(bDefaultParent=True, Child=@DISPATCH_RigVMDispatch_ArrayIterator.Element) {is=struct RigVMUnitNode}.Parent
   )
   ItemArray_1 = RigVMDispatchNode(Value=((Type=Bone,Name="upperarm_l"),(Type=Bone,Name="upperarm_r"),(Type=Bone,Name="hand_l"),(Type=Bone,Name="hand_r"),(Type=Bone,Name="thigh_l"),(Type=Bone,Name="thigh_r"),(Type=Bone,Name="foot_l"),(Type=Bone,Name="foot_r"))) {template="DISPATCH_RigVMDispatch_Constant(io Value)"}
   RigVMDispatchNode(ExecuteContext=(), Other=(), Array=@ItemArray_1.Value) {template="DISPATCH_RigVMDispatch_ArrayUnion(io Array,in Other)"}
  ]}
]
hierarchy [
 root {loc=<0 0 0>, rot=<0 0 0 1>}
 pelvis {parent=root, loc=<0 -0.7071 93.1191>, rot=<0.0127 0.707 -0.0127 -0.707>}
 spine_01 {parent=pelvis, loc=<6.7494 -0.6388 0>, rot=<0 0 0.034 -0.9994>}
 spine_02 {parent=spine_01, loc=<7.2518 -0.035 0>, rot=<0 0 0.0238 -0.9997>}
 spine_03 {parent=spine_02, loc=<6.0104 0 0>, rot=<0 0 0.0149 -0.9999>}
 spine_04 {parent=spine_03, loc=<8.1024 0 0>, rot=<0 0 -0.0132 -0.9999>}
 spine_05 {parent=spine_04, loc=<10.0146 0 0>, rot=<0 0 -0.0015 -1>}
 neck_01 {parent=spine_05, loc=<11.3138 0.1529 0>, rot=<0 0 -0.2003 -0.9797>}
 neck_02 {parent=neck_01, loc=<4.5627 0 0>, rot=<0 0 0.0504 -0.9987>}
 head {parent=neck_02, loc=<5.3914 0 0>, rot=<0 0 0.115 -0.9934>}
 kin_hoodSide_l {parent=head, loc=<4.8131 -8.019 -6.7625>, rot=<-0.5512 -0.8217 0.0451 -0.138>}
 kin_hoodSide_r {parent=head, loc=<4.1166 -5.498 9.7738>, rot=<-0.0773 -0.052 -0.8306 -0.5491>}
 kin_hood_def_1_c {parent=head, loc=<3.8315 -11.1303 -0.0017>, rot=<0.0029 -0.7071 -0.0029 -0.7071>}
 kin_hood_def_1_l {parent=head, loc=<2.3098 -6.4973 -10.4683>, rot=<0.0029 -0.7071 -0.0029 -0.7071>}
 kin_hood_def_1_r {parent=head, loc=<2.2259 -6.2655 10.2787>, rot=<0.0029 -0.7071 -0.0029 -0.7071>}
 base_hoodTop {parent=head, loc=<16.7936 0.1619 0>, rot=<0.4317 0.56 0.5601 0.4318>}
 dyn_hoodTop {parent=base_hoodTop, loc=<0 0 0>, rot=<0 0 0 1>}
 kin_hood_def_2 {parent=neck_02, loc=<-0.2613 -10.2293 -0.001>, rot=<-0.0784 -0.7027 0.0784 -0.7027>}
 kin_hood_def_2_l {parent=neck_02, loc=<-1.1333 8.8207 -6.9103>, rot=<-0.0784 -0.7027 0.0784 -0.7027>}
 kin_hood_def_2_r {parent=neck_02, loc=<-1.3601 8.8123 7.0337>, rot=<-0.0784 -0.7027 0.0784 -0.7027>}
 attach_cape {parent=spine_05, loc=<9.017 -11.714 0>, rot=<0.0616 -0.7044 -0.0616 -0.7044>}
 cape {parent=attach_cape, type=Null, loc=<0 0.36 -5>, rot=<0 0 0 1>}
 attach_backpack {parent=spine_05, loc=<-0.123 -13.473 0>, rot=<-0.0093 -0.707 0.0093 -0.707>}
 Backpack_BR {parent=attach_backpack, type=Null, loc=<0 0.44 -2>, rot=<-0.061 0 0 0.9981>}
 clavicle_l {parent=spine_05, loc=<4.3575 -0.1466 -2.3223>, rot=<0.0465 -0.7289 -0.009 -0.6829>}
 upperarm_l {parent=clavicle_l, loc=<15.5111 0 0>, rot=<0.0226 -0.3234 -0.0088 -0.9459>}
 lowerarm_l {parent=upperarm_l, loc=<26.6514 0 0>, rot=<0.0391 -0.0099 -0.2453 -0.9686>}
 hand_l {parent=lowerarm_l, loc=<22.3491 0 0>, rot=<-0.7081 -0.0507 0.0975 -0.6975>}
 thumb_01_l {parent=hand_l, loc=<1.3118 -0.3191 -2.459>, rot=<0.6546 -0.4042 0.0664 -0.6354>}
 thumb_02_l {parent=thumb_01_l, loc=<5.9336 -0.0448 0.0401>, rot=<-0.0037 0.0213 0.298 -0.9543>}
 thumb_03_l {parent=thumb_02_l, loc=<2.2223 0.3682 -0.0907>, rot=<0.0118 -0.013 0.0066 -0.9998>}
 index_metacarpal_l {parent=hand_l, loc=<1.6535 0.0183 -2.96>, rot=<0.0158 -0.0124 -0.0173 -0.9996>}
 index_01_l {parent=index_metacarpal_l, loc=<8.3008 0.0357 -0.0524>, rot=<-0.0044 0.0485 0.0924 -0.9945>}
 index_02_l {parent=index_01_l, loc=<3.3587 -0.3004 0.0452>, rot=<-0.002 0.0027 0.3225 -0.9466>}
 index_03_l {parent=index_02_l, loc=<2.5578 -0.123 -0.0252>, rot=<-0.002 0.0067 0.2558 -0.9667>}
 middle_metacarpal_l {parent=hand_l, loc=<2.1856 -0.1725 -0.5984>, rot=<0.006 0.0168 -0.0329 -0.9993>}
 middle_01_l {parent=middle_metacarpal_l, loc=<7.3275 -0.2447 0.0129>, rot=<-0.0208 0.0042 0.0858 -0.9961>}
 middle_02_l {parent=middle_01_l, loc=<4.0123 0.1401 0.1175>, rot=<-0.0218 0.0258 0.3512 -0.9357>}
 middle_03_l {parent=middle_02_l, loc=<2.6077 0.0534 0.1373>, rot=<0.0082 0.0188 0.1708 -0.9851>}
 ring_metacarpal_l {parent=hand_l, loc=<2.2074 -0.3625 1.3638>, rot=<-0.0284 0.0722 -0.0458 -0.9959>}
 ring_01_l {parent=ring_metacarpal_l, loc=<7.0295 -0.6197 -0.1446>, rot=<-0.0489 -0.0252 0.1892 -0.9804>}
 ring_02_l {parent=ring_01_l, loc=<3.6174 0.4647 -0.1027>, rot=<0 0 0.2806 -0.9598>}
 ring_03_l {parent=ring_02_l, loc=<2.4152 0.114 0.0199>, rot=<0 0 0.1189 -0.9929>}
 pinky_metacarpal_l {parent=hand_l, loc=<1.7616 -0.9753 3.3814>, rot=<-0.0397 0.0814 -0.069 -0.9935>}
 pinky_01_l {parent=pinky_metacarpal_l, loc=<7.1709 -0.4831 -0.1094>, rot=<-0.0654 -0.0035 0.2242 -0.9723>}
 pinky_02_l {parent=pinky_01_l, loc=<2.8274 0.0893 -0.1043>, rot=<0.0037 0.0011 0.2854 -0.9584>}
 pinky_03_l {parent=pinky_02_l, loc=<2.2046 0.0307 -0.0078>, rot=<-0.0002 0.0098 0.1445 -0.9895>}
 deform_wrist_dn_base_l {parent=hand_l, loc=<0 -2.2342 0>, rot=<0 0 0 -1>}
 deform_wrist_dn_l {parent=deform_wrist_dn_base_l, loc=<0 0 0>, rot=<0 0 0 1>}
 deform_wrist_up_base_l {parent=hand_l, loc=<0 2.2682 0>, rot=<0 0 0 -1>}
 deform_wrist_up_l {parent=deform_wrist_up_base_l, loc=<0 0 0>, rot=<0 0 0 1>}
 lowerarm_twist_02_l {parent=lowerarm_l, loc=<9.6199 0 0>, rot=<0 0 0 -1>}
 lowerarm_twist_01_l {parent=lowerarm_l, loc=<17.8549 0 0>, rot=<0 0 0 -1>}
 deform_elbow_in_base_l {parent=lowerarm_l, loc=<0.2112 5.259 0.8274>, rot=<0 0 0 -1>}
 deform_elbow_in_l {parent=deform_elbow_in_base_l, loc=<0 0 0>, rot=<0 0 0 1>}
 deform_elbow_out_base_l {parent=lowerarm_l, loc=<0 -4 0>, rot=<0 0 0 -1>}
 deform_elbow_out_l {parent=deform_elbow_out_base_l, loc=<0 0 0>, rot=<0 0 0 1>}
 upperarm_twist_01_l {parent=upperarm_l, loc=<0.0041 0 0>, rot=<0 0 0 -1>}
 deform_upperarm_bk_base_l {parent=upperarm_twist_01_l, loc=<1.0813 -7.2193 -1.9453>, rot=<0 0 0 1>}
 deform_upperarm_bk_l {parent=deform_upperarm_bk_base_l, loc=<0 0 0>, rot=<0 0 0 1>}
 deform_upperarm_fr_base_l {parent=upperarm_twist_01_l, loc=<-0.3691 7.0506 -2.765>, rot=<0 0 0 1>}
 deform_upperarm_fr_l {parent=deform_upperarm_fr_base_l, loc=<0 0 0>, rot=<0 0 0 1>}
 deform_upperarm_up_base_l {parent=upperarm_twist_01_l, loc=<-3.2216 -1.1391 3.3311>, rot=<0 0 0 1>}
 deform_upperarm_up_l {parent=deform_upperarm_up_base_l, loc=<0 0 0>, rot=<0 0 0 1>}
 coll_upperarm_twist_01_l {parent=upperarm_twist_01_l, loc=<0 0 0>, rot=<0 0 0 1>}
 upperarm_twist_02_l {parent=upperarm_l, loc=<13.7941 0 0>, rot=<0 0 0 -1>}
 clavicle_r {parent=spine_05, loc=<4.3575 -0.1466 2.3223>, rot=<0.7289 0.0465 -0.6829 0.009>}
 upperarm_r {parent=clavicle_r, loc=<-15.5111 0 0>, rot=<0.0226 -0.3234 -0.0088 -0.9459>}
 lowerarm_r {parent=upperarm_r, loc=<-26.6514 0 0>, rot=<0.0391 -0.0099 -0.2453 -0.9686>}
 hand_r {parent=lowerarm_r, loc=<-22.3491 0 0>, rot=<-0.7081 -0.0507 0.0975 -0.6975>}
 thumb_01_r {parent=hand_r, loc=<-1.3118 0.3191 2.459>, rot=<0.6546 -0.4042 0.0664 -0.6354>}
 thumb_02_r {parent=thumb_01_r, loc=<-5.9336 0.0448 -0.0401>, rot=<-0.0037 0.0213 0.298 -0.9543>}
 thumb_03_r {parent=thumb_02_r, loc=<-2.2223 -0.3682 0.0907>, rot=<0.0118 -0.013 0.0066 -0.9998>}
 index_metacarpal_r {parent=hand_r, loc=<-1.6535 -0.0183 2.96>, rot=<0.0158 -0.0124 -0.0173 -0.9996>}
 index_01_r {parent=index_metacarpal_r, loc=<-8.3008 -0.0357 0.0524>, rot=<-0.0044 0.0485 0.0924 -0.9945>}
 index_02_r {parent=index_01_r, loc=<-3.3587 0.3004 -0.0452>, rot=<-0.002 0.0027 0.3225 -0.9466>}
 index_03_r {parent=index_02_r, loc=<-2.5578 0.123 0.0252>, rot=<-0.002 0.0067 0.2558 -0.9667>}
 middle_metacarpal_r {parent=hand_r, loc=<-2.1856 0.1725 0.5984>, rot=<0.006 0.0168 -0.0329 -0.9993>}
 middle_01_r {parent=middle_metacarpal_r, loc=<-7.3275 0.2447 -0.0129>, rot=<-0.0208 0.0042 0.0858 -0.9961>}
 middle_02_r {parent=middle_01_r, loc=<-4.0123 -0.1401 -0.1175>, rot=<-0.0218 0.0258 0.3512 -0.9357>}
 middle_03_r {parent=middle_02_r, loc=<-2.6077 -0.0534 -0.1373>, rot=<0.0082 0.0188 0.1708 -0.9851>}
 ring_metacarpal_r {parent=hand_r, loc=<-2.2074 0.3625 -1.3638>, rot=<-0.0284 0.0722 -0.0458 -0.9959>}
 ring_01_r {parent=ring_metacarpal_r, loc=<-7.0295 0.6197 0.1446>, rot=<-0.0489 -0.0252 0.1892 -0.9804>}
 ring_02_r {parent=ring_01_r, loc=<-3.6174 -0.4647 0.1027>, rot=<0 0 0.2806 -0.9598>}
 ring_03_r {parent=ring_02_r, loc=<-2.4152 -0.114 -0.0199>, rot=<0 0 0.1189 -0.9929>}
 pinky_metacarpal_r {parent=hand_r, loc=<-1.7616 0.9753 -3.3814>, rot=<-0.0397 0.0814 -0.069 -0.9935>}
 pinky_01_r {parent=pinky_metacarpal_r, loc=<-7.1709 0.4831 0.1094>, rot=<-0.0654 -0.0035 0.2242 -0.9723>}
 pinky_02_r {parent=pinky_01_r, loc=<-2.8274 -0.0893 0.1043>, rot=<0.0037 0.0011 0.2854 -0.9584>}
 pinky_03_r {parent=pinky_02_r, loc=<-2.2046 -0.0307 0.0078>, rot=<-0.0002 0.0098 0.1445 -0.9895>}
 deform_wrist_dn_base_r {parent=hand_r, loc=<-0.0004 2.2345 -0.0001>, rot=<0 0 0 -1>}
 deform_wrist_dn_r {parent=deform_wrist_dn_base_r, loc=<0 0 0>, rot=<0 0 0 1>}
 deform_wrist_up_base_r {parent=hand_r, loc=<0.0003 -2.2684 0.0001>, rot=<0 0 0 -1>}
 deform_wrist_up_r {parent=deform_wrist_up_base_r, loc=<0 0 0>, rot=<0 0 0 1>}
 lowerarm_twist_02_r {parent=lowerarm_r, loc=<-9.6199 0 0>, rot=<0 0 0 -1>}
 lowerarm_twist_01_r {parent=lowerarm_r, loc=<-17.8549 0 0>, rot=<0 0 0 -1>}
 deform_elbow_in_base_r {parent=lowerarm_r, loc=<-0.2111 -5.2591 -0.8275>, rot=<0 0 0 -1>}
 deform_elbow_in_r {parent=deform_elbow_in_base_r, loc=<-0.0001 0 0.0002>, rot=<0 0 0 1>}
 deform_elbow_out_base_r {parent=lowerarm_r, loc=<0 4 -0.0001>, rot=<0 0 0 -1>}
 deform_elbow_out_r {parent=deform_elbow_out_base_r, loc=<0 0 0>, rot=<0 0 0 1>}
 upperarm_twist_01_r {parent=upperarm_r, loc=<-0.0041 0 0>, rot=<0 0 0 -1>}
 deform_upperarm_bk_base_r {parent=upperarm_twist_01_r, loc=<-1.0814 7.2192 1.9455>, rot=<0 0 0 1>}
 deform_upperarm_bk_r {parent=deform_upperarm_bk_base_r, loc=<0 0 0>, rot=<0 0 0 1>}
 deform_upperarm_fr_base_r {parent=upperarm_twist_01_r, loc=<0.3689 -7.0506 2.7652>, rot=<0 0 0 1>}
 deform_upperarm_fr_r {parent=deform_upperarm_fr_base_r, loc=<0 0 0>, rot=<0 0 0 1>}
 deform_upperarm_up_base_r {parent=upperarm_twist_01_r, loc=<3.2214 1.1391 -3.3308>, rot=<0 0 0 1>}
 deform_upperarm_up_r {parent=deform_upperarm_up_base_r, loc=<0 0 0>, rot=<0 0 0 -1>}
 coll_upperarm_twist_01_r {parent=upperarm_twist_01_r, loc=<0 0 0>, rot=<0 0 0 1>}
 upperarm_twist_02_r {parent=upperarm_r, loc=<-13.7941 0 0>, rot=<0 0 0 -1>}
 base_jacketCollarStand_1_l {parent=spine_05, loc=<5.7733 11.1361 -5.6055>, rot=<0.908 0.1207 -0.4007 -0.0208>}
 dyn_jacketCollarStand_1_l {parent=base_jacketCollarStand_1_l, loc=<0 0 0>, rot=<0 0 0 1>}
 base_jacketCollarStand_2_l {parent=spine_05, loc=<10.9951 3.084 -10.1429>, rot=<0.9433 0.1495 -0.2944 0.0353>}
 dyn_jacketCollarStand_2_l {parent=base_jacketCollarStand_2_l, loc=<0 0 0>, rot=<0 0 0 1>}
 base_jacketCollarStand_3_l {parent=spine_05, loc=<12.9358 -3.8601 -8.8714>, rot=<0.9376 0.0572 -0.172 0.2967>}
 dyn_jacketCollarStand_3_l {parent=base_jacketCollarStand_3_l, loc=<0 0 0>, rot=<0 0 0 1>}
 base_jacketCollarStand_1_c {parent=spine_05, loc=<13.1203 -7.4258 0>, rot=<-0.7003 0.0982 0.0982 -0.7003>}
 dyn_jacketCollarStand_1_c {parent=base_jacketCollarStand_1_c, loc=<0 0 0>, rot=<0 0 0 1>}
 base_jacketCollarStand_1_r {parent=spine_05, loc=<6.3546 10.5715 7.0801>, rot=<0.0368 0.3575 -0.1404 -0.9226>}
 dyn_jacketCollarStand_1_r {parent=base_jacketCollarStand_1_r, loc=<0 0 0>, rot=<0 0 0 1>}
 base_jacketCollarStand_2_r {parent=spine_05, loc=<10.7348 3.0101 10.5368>, rot=<-0.0302 0.262 -0.1506 -0.9528>}
 dyn_jacketCollarStand_2_r {parent=base_jacketCollarStand_2_r, loc=<0 0 0>, rot=<0 0 0 1>}
 base_jacketCollarStand_3_r {parent=spine_05, loc=<13.0006 -3.7308 8.6998>, rot=<-0.297 0.1767 -0.0557 -0.9367>}
 dyn_jacketCollarStand_3_r {parent=base_jacketCollarStand_3_r, loc=<0 0 0>, rot=<0 0 0 1>}
 deform_pec_aim_base_l {parent=spine_04, loc=<7.8732 12.6388 -2.4123>, rot=<0.0102 -0.4602 0.3979 -0.7936>}
 deform_pec_aim_l {parent=deform_pec_aim_base_l, loc=<0 0 0>, rot=<0 0 0 1>}
 deform_pec_base_l {parent=deform_pec_aim_l, loc=<10.766 0 0>, rot=<0 0 0 1>}
 deform_pec_l {parent=deform_pec_base_l, loc=<0 0 0>, rot=<0 0 0 1>}
 deform_pec_aim_base_r {parent=spine_04, loc=<7.8727 12.6388 2.4123>, rot=<-0.4602 -0.0102 0.7936 0.3979>}
 deform_pec_aim_r {parent=deform_pec_aim_base_r, loc=<0 0 0>, rot=<0 0 0 1>}
 deform_pec_base_r {parent=deform_pec_aim_r, loc=<-10.7662 -0.0006 -0.0007>, rot=<0 0 0 1>}
 deform_pec_r {parent=deform_pec_base_r, loc=<0 0 0>, rot=<0 0 0 1>}
 deform_lat_base_l {parent=spine_04, loc=<-0.1102 0.3051 -15.4494>, rot=<0.0294 -0.7065 -0.0294 -0.7065>}
 deform_lat_l {parent=deform_lat_base_l, loc=<0 0 0>, rot=<0 0 0 1>}
 deform_lat_base_r {parent=spine_04, loc=<-0.11 0.3046 15.4494>, rot=<0.7065 0.0294 0.7065 -0.0294>}
 deform_lat_r {parent=deform_lat_base_r, loc=<0 0 0>, rot=<0 0 0 1>}
 kin_hemJacket {parent=spine_02, loc=<0 0 0>, rot=<0 0 0 1>}
 base_hemJacket_1_r {parent=kin_hemJacket, loc=<1.1219 13.0098 9.0976>, rot=<0.0283 -0.2471 0.9685 -0.0068>}
 dyn_hemJacket_1_r {parent=base_hemJacket_1_r, loc=<0 0 0>, rot=<0 0 0 1>}
 base_hemJacket_1_l {parent=kin_hemJacket, loc=<1.1224 13.0099 -9.0976>, rot=<-0.0283 0.2471 0.9685 -0.0068>}
 dyn_hemJacket_1_l {parent=base_hemJacket_1_l, loc=<0 0 0>, rot=<0 0 0 1>}
 base_hemJacket_2_r {parent=kin_hemJacket, loc=<1.8724 5.6926 16.8384>, rot=<0.0707 -0.5709 0.8179 0.0098>}
 dyn_hemJacket_2_r {parent=base_hemJacket_2_r, loc=<0 0 0>, rot=<0 0 0 1>}
 base_hemJacket_2_l {parent=kin_hemJacket, loc=<1.8725 5.6926 -16.8384>, rot=<-0.0707 0.5709 0.8179 0.0098>}
 dyn_hemJacket_2_l {parent=base_hemJacket_2_l, loc=<0 0 0>, rot=<0 0 0 1>}
 base_hemJacket_3_r {parent=kin_hemJacket, loc=<2.7438 -6.4373 15.7136>, rot=<0.0567 -0.85 0.5235 0.0163>}
 dyn_hemJacket_3_r {parent=base_hemJacket_3_r, loc=<0 0 0>, rot=<0 0 0 1>}
 base_hemJacket_3_l {parent=kin_hemJacket, loc=<2.7435 -6.4373 -15.7136>, rot=<0.0567 -0.85 -0.5235 -0.0163>}
 dyn_hemJacket_3_l {parent=base_hemJacket_3_l, loc=<0 0 0>, rot=<0 0 0 1>}
 base_hemJacket_1_c {parent=kin_hemJacket, loc=<2.741 -12.3031 -0.4968>, rot=<0.0399 -0.9992 0 0>}
 dyn_hemJacket_1_c {parent=base_hemJacket_1_c, loc=<0 0 0>, rot=<0 0 0 1>}
 kin_delta_hemJacket_1_r {parent=kin_hemJacket, loc=<1.6211 13.0234 9.0719>, rot=<0.0283 -0.2471 0.9685 -0.0068>}
 kin_delta_hemJacket_1_l {parent=kin_hemJacket, loc=<1.6216 13.0234 -9.0719>, rot=<-0.0283 0.2471 0.9685 -0.0068>}
 thigh_l {parent=pelvis, loc=<-2.5693 -0.0109 -7.7501>, rot=<0.0012 0.0572 0.0155 -0.9982>}
 calf_l {parent=thigh_l, loc=<-39.0484 0 0>, rot=<0.0006 -0.0161 -0.0391 -0.9991>}
 foot_l {parent=calf_l, loc=<-42.561 0.1077 0.4292>, rot=<0.0418 -0.0455 0.0343 -0.9975>}
 ball_l {parent=foot_l, loc=<-6.1599 13.8264 0.6806>, rot=<0.0001 -0.0001 -0.7063 -0.7079>}
 deform_ankle_in_base_l {parent=foot_l, loc=<2.9481 5.9274 0.65>, rot=<-0.0253 0.0419 -0.526 -0.8491>}
 deform_ankle_in_l {parent=deform_ankle_in_base_l, loc=<0 0 0>, rot=<0 0 0 1>}
 deform_ankle_out_base_l {parent=foot_l, loc=<-2.4371 -4.9 -0.5374>, rot=<-0.0253 0.0419 -0.526 -0.8491>}
 deform_ankle_out_l {parent=deform_ankle_out_base_l, loc=<0 0 0>, rot=<0 0 0 1>}
 calf_twist_02_l {parent=calf_l, loc=<-18.8064 -0.0967 0.0397>, rot=<0.0237 0.0095 -0.0035 -0.9997>}
 calf_twist_01_l {parent=calf_l, loc=<-35.813 -0.0967 0.0397>, rot=<0.0237 0.0095 -0.0035 -0.9997>}
 deform_knee_in_base_l {parent=calf_l, loc=<1.298 -5.2957 -0.0496>, rot=<0 0 0 -1>}
 deform_knee_in_l {parent=deform_knee_in_base_l, loc=<0 0 0>, rot=<0 0 0 1>}
 deform_knee_out_base_l {parent=calf_l, loc=<0 4 0>, rot=<0 0 0 -1>}
 deform_knee_out_l {parent=deform_knee_out_base_l, loc=<0 0 0>, rot=<0 0 0 1>}
 calf_twist_00_l {parent=calf_l, loc=<0 0 0>, rot=<0 0 0 1>}
 thigh_twist_01_l {parent=thigh_l, loc=<-21.6303 0 0>, rot=<0.0013 0 0 -1>}
 thigh_tshirt_l {parent=thigh_twist_01_l, loc=<20.6303 0 0>, rot=<-0.0013 0 0 -1>}
 deform_groin_base_l {parent=thigh_l, loc=<-2.1908 8.1142 -1.7442>, rot=<0 0 0 -1>}
 deform_groin_l {parent=deform_groin_base_l, loc=<0 0 0>, rot=<0 0 0 1>}
 thigh_twist_00_l {parent=thigh_l, loc=<-24.348 0 0>, rot=<0.0013 0 0 -1>}
 base_strapHang_out_l {parent=thigh_twist_00_l, loc=<-0.0785 0.9313 -0.9309>, rot=<-0.0028 -0.9153 -0.0002 -0.4027>}
 dyn_strapHang_out_l {parent=base_strapHang_out_l, loc=<0 0 0>, rot=<0 0 0 1>}
 base_strapHang_in_l {parent=thigh_twist_00_l, loc=<-0.2906 0.9234 2.3819>, rot=<0.3453 -0.0021 0.9385 0.0018>}
 dyn_strapHang_in_l {parent=base_strapHang_in_l, loc=<0 0 0>, rot=<0 0 0 1>}
 kin_strap_frontTris_l {parent=thigh_twist_00_l, loc=<1.2 10 1.2>, rot=<0 0 0 1>}
 target_strap_01_l {parent=thigh_l, loc=<-25.6021 10.4002 0.6359>, rot=<0 0 0 1>}
 target_strap_02_l {parent=thigh_l, loc=<-32.9711 3.8803 -9.6501>, rot=<0 0 0 1>}
 coll_strap_01_l {parent=thigh_l, loc=<-8.3471 10.8121 3.5943>, rot=<0.0117 -0.9963 0.0013 -0.0848>}
 coll_strap_02_l {parent=thigh_l, loc=<-23.9279 10.1298 -0.7387>, rot=<0.2091 -0.901 0.0877 -0.3698>}
 thigh_r {parent=pelvis, loc=<-2.5693 -0.0109 7.7501>, rot=<0.0572 -0.0012 0.9982 0.0155>}
 calf_r {parent=thigh_r, loc=<39.0484 0 0>, rot=<0.0006 -0.0161 -0.0391 -0.9991>}
 foot_r {parent=calf_r, loc=<42.561 -0.1077 -0.4292>, rot=<0.0418 -0.0455 0.0343 -0.9975>}
 ball_r {parent=foot_r, loc=<6.1599 -13.8264 -0.6806>, rot=<0.0001 -0.0001 -0.7063 -0.7079>}
 deform_ankle_in_base_r {parent=foot_r, loc=<-2.9481 -5.9274 -0.6501>, rot=<-0.0253 0.0419 -0.526 -0.8491>}
 deform_ankle_in_r {parent=deform_ankle_in_base_r, loc=<0 0 0>, rot=<0 0 0 1>}
 deform_ankle_out_base_r {parent=foot_r, loc=<2.4371 4.9 0.5373>, rot=<-0.0253 0.0419 -0.526 -0.8491>}
 deform_ankle_out_r {parent=deform_ankle_out_base_r, loc=<0 0 0>, rot=<0 0 0 1>}
 calf_twist_02_r {parent=calf_r, loc=<18.8064 0.0967 -0.0397>, rot=<0.0237 0.0095 -0.0035 -0.9997>}
 base_calfLace_1_l {parent=calf_twist_02_r, loc=<12.3206 4.037 4.6723>, rot=<0.8692 -0.099 0.3205 -0.3632>}
 dyn_calfLace_1_l {parent=base_calfLace_1_l, loc=<0 0 0>, rot=<0 0 0 1>}
 dyn_calfLace_2_l {parent=dyn_calfLace_1_l, loc=<3.623 0 0>, rot=<0.0857 0.1663 0.0775 -0.9793>}
 calf_twist_01_r {parent=calf_r, loc=<35.813 0.0967 -0.0397>, rot=<0.0237 0.0095 -0.0035 -0.9997>}
 deform_knee_in_base_r {parent=calf_r, loc=<-1.298 5.2957 0.0496>, rot=<0 0 0 -1>}
 deform_knee_in_r {parent=deform_knee_in_base_r, loc=<0 0 0>, rot=<0 0 0 1>}
 deform_knee_out_base_r {parent=calf_r, loc=<0 -4 0>, rot=<0 0 0 -1>}
 deform_knee_out_r {parent=deform_knee_out_base_r, loc=<0 0 0>, rot=<0 0 0 1>}
 calf_twist_00_r {parent=calf_r, loc=<0 0 0>, rot=<0 0 0 1>}
 thigh_twist_01_r {parent=thigh_r, loc=<21.6303 0 0>, rot=<0.0013 0 0 -1>}
 thigh_tshirt_r {parent=thigh_twist_01_r, loc=<-20.6303 0 0>, rot=<-0.0013 0 0 -1>}
 deform_groin_base_r {parent=thigh_r, loc=<2.1908 -8.1142 1.7442>, rot=<0 0 0 -1>}
 deform_groin_r {parent=deform_groin_base_r, loc=<0 0 0>, rot=<0 0 0 1>}
 thigh_twist_00_r {parent=thigh_r, loc=<24.348 0 0>, rot=<0.0013 0 0 -1>}
 base_strapHang_out_r {parent=thigh_twist_00_r, loc=<0.0785 -0.9313 0.9309>, rot=<-0.0002 0.4027 0.0028 -0.9153>}
 dyn_strapHang_out_r {parent=base_strapHang_out_r, loc=<0 0 0>, rot=<0 0 0 1>}
 base_strapHang_in_r {parent=thigh_twist_00_r, loc=<0.2906 -0.9234 -2.3818>, rot=<0.9526 -0.0019 -0.3042 -0.002>}
 dyn_strapHang_in_r {parent=base_strapHang_in_r, loc=<0 0 0>, rot=<0 0 0 1>}
 kin_strap_frontTris_r {parent=thigh_twist_00_r, loc=<-1.2 -10 -1.2>, rot=<0 0 0 1>}
 deform_glute_base_l {parent=pelvis, loc=<-8.8064 -8.8222 -8.4287>, rot=<0.0275 0.0501 -0.4472 -0.8926>}
 deform_glute_l {parent=deform_glute_base_l, loc=<0 0 0>, rot=<0 0 0 1>}
 deform_glute_base_r {parent=pelvis, loc=<-8.8062 -8.8222 8.4287>, rot=<0.0501 -0.0275 0.8926 -0.4472>}
 deform_glute_r {parent=deform_glute_base_r, loc=<0 0 0>, rot=<0 0 0 1>}
 base_hipChains {parent=pelvis, loc=<-3.9069 3.4534 12.7954>, rot=<0.0075 -0.975 0.1399 0.1726>}
 dyn_hipChains {parent=base_hipChains, loc=<0 0 0>, rot=<0 0 0 1>}
 dyn_hipChains_front {parent=dyn_hipChains, loc=<8.2695 7.1653 1.0993>, rot=<0 0 0 1>}
 dyn_hipChains_back {parent=dyn_hipChains, loc=<9.5755 -7.7709 -0.2513>, rot=<0 0 0 1>}
 kin_carabiners_01_r {parent=pelvis, loc=<-0.1112 9.6521 11.9341>, rot=<-0.2431 -0.6744 -0.1867 -0.6718>}
 kin_hoodie_hem_01_r {parent=pelvis, loc=<5.885 7.3222 12.7507>, rot=<-0.0127 -0.707 0.0127 -0.707>}
 kin_carabiners_01_l {parent=pelvis, loc=<-0.3502 8.6716 -13.0299>, rot=<0.6951 0.13 0.6899 0.1548>}
 base_hangCarabinerF_1 {parent=kin_carabiners_01_l, loc=<-0.0253 -0.2536 -4.1373>, rot=<-0.4976 -0.4826 0.5393 -0.4782>}
 dyn_hangCarabinerF_1 {parent=base_hangCarabinerF_1, loc=<0 0 0>, rot=<0 0 0 1>}
 dyn_hangCarabinerF_2 {parent=dyn_hangCarabinerF_1, loc=<5.4674 0 0>, rot=<-0.0399 0.0392 -0.0016 -0.9984>}
 kin_carabiners_02_l {parent=pelvis, loc=<-0.1246 5.3503 -16.0407>, rot=<0.6206 0.3388 0.6081 0.3608>}
 base_hangCarabinerS_1 {parent=kin_carabiners_02_l, loc=<0.174 -0.6164 -3.8873>, rot=<-0.4044 -0.5523 0.5061 -0.5246>}
 dyn_hangCarabinerS_1 {parent=base_hangCarabinerS_1, loc=<0 0 0>, rot=<0 0 0 1>}
 dyn_hangCarabinerS_2 {parent=dyn_hangCarabinerS_1, loc=<6.6494 -0.0053 -0.0157>, rot=<0 0.0633 0 -0.998>}
 coll_hem_f_r {parent=pelvis, loc=<-19.5365 5.0036 17.6577>, rot=<-0.4365 -0.3413 -0.1164 -0.8243>}
 target_hem_f_r {parent=pelvis, loc=<-11.3146 10.4464 12.535>, rot=<-0.0127 -0.707 0.0127 -0.707>}
 kin_strapStart_l {parent=pelvis, loc=<-2.7956 10.9807 -3.1745>, rot=<0 0 0 1>}
 kin_strapStart_r {parent=pelvis, loc=<-2.7956 10.9807 3.3659>, rot=<0 0 0 1>}
]
variables [
 IsTargeting {type=bool, default=False}
 isUsingPistol {type=bool, default=False}
]
