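# Animation Blueprint for the face accessory of the male IvoryDwell Aces body, adding secondary motion to five hair strands.
# It copies the attached parent's pose, passes it through an additive head bone transform, then chains spring-driven, angle-limited Anim Dynamics nodes per strand.
# Each update stores an inverted, scaled head-versus-neck_02 reference-pose delta in headRot, with yaw weighted by twist.
# The head transform shows zero rotation, with no visible read of headRot.
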AnimBlueprint /BRCosmetics/Characters/Player/Male/Medium/Bodies/M_MED_IvoryDwell_Aces/Meshes/Parts/M_MED_IvoryDwell_Aces_FaceAcc_AnimBP
target_skeleton=&/M_MED_IvoryDwell_Aces_FaceAcc_Skeleton
graphs [
 AnimGraph {nodes [
   n1 = "Transform (Modify) Bone - Bone: head"!(ComponentPose=Local To Component(LocalPose=Copy Pose From Mesh() {bUseAttachedParent=true}), Alpha=1, Rotation="0,0,0") -> (Pose) {
    rotationMode=BMM_Additive
    rotationSpace=BCS_BoneSpace
    translationMode=BMM_Additive
    translationSpace=BCS_BoneSpace
    alphaBoolBlend {blendOption=Linear}
    boneToModify {boneName=head}
   }
   n2 = Anim Dynamics - Bone: dyn_hairStrand_1!(ComponentPose=@n1, Alpha=0.75) -> (Pose) {
    alpha=0.75
    angularDampingOverride=1
    angularSpringConstant=12000
    bAngularSpring=true
    bOverrideAngularDamping=true
    gravityScale=0
    simulationSpace=RootRelative
    boundBone {boneName=dyn_hairStrand_1}
    componentAppliedLinearAccClamp=<4000 4000 4000>
    componentLinearAccScale=<0.4 0.4 0.4>
    componentLinearVelScale=<0.4 0.4 0.4>
    gravityOverride {z=-980}
    physicsBodyDefinitions [{
      boundBone {boneName=dyn_hairStrand_1}
      boxExtents=<20 30 30>
      constraintSetup {angularLimitsMax=<0 12 12>, angularLimitsMin=<0 -12 -12>, angularTarget=<1 0 0>}
      localJointOffset=<2 0 0>
     }]
   }
   n3 = Anim Dynamics - Bone: dyn_hairStrand_2!(ComponentPose=@n2, Alpha=0.75) -> (Pose) {
    alpha=0.75
    angularDampingOverride=1
    angularSpringConstant=12000
    bAngularSpring=true
    bOverrideAngularDamping=true
    gravityScale=0
    simulationSpace=RootRelative
    boundBone {boneName=dyn_hairStrand_2}
    componentAppliedLinearAccClamp=<4000 4000 4000>
    componentLinearAccScale=<0.4 0.4 0.4>
    componentLinearVelScale=<0.4 0.4 0.4>
    gravityOverride {z=-980}
    physicsBodyDefinitions [{
      boundBone {boneName=dyn_hairStrand_2}
      boxExtents=<20 30 30>
      constraintSetup {angularLimitsMax=<0 12 12>, angularLimitsMin=<0 -12 -12>, angularTarget=<1 0 0>}
      localJointOffset=<2 0 0>
     }]
   }
   n4 = Anim Dynamics - Bone: dyn_hairStrand_3!(ComponentPose=@n3, Alpha=0.75) -> (Pose) {
    alpha=0.75
    angularDampingOverride=1
    angularSpringConstant=12000
    bAngularSpring=true
    bOverrideAngularDamping=true
    gravityScale=0
    simulationSpace=RootRelative
    boundBone {boneName=dyn_hairStrand_3}
    componentAppliedLinearAccClamp=<4000 4000 4000>
    componentLinearAccScale=<0.4 0.4 0.4>
    componentLinearVelScale=<0.4 0.4 0.4>
    gravityOverride {z=-980}
    physicsBodyDefinitions [{
      boundBone {boneName=dyn_hairStrand_3}
      boxExtents=<20 30 30>
      constraintSetup {angularLimitsMax=<0 12 12>, angularLimitsMin=<0 -12 -12>, angularTarget=<1 0 0>}
      localJointOffset=<2 0 0>
     }]
   }
   n5 = Anim Dynamics - Bone: dyn_hairStrand_4!(ComponentPose=@n4, Alpha=0.75) -> (Pose) {
    alpha=0.75
    angularDampingOverride=1
    angularSpringConstant=12000
    bAngularSpring=true
    bOverrideAngularDamping=true
    gravityScale=0
    simulationSpace=RootRelative
    boundBone {boneName=dyn_hairStrand_4}
    componentAppliedLinearAccClamp=<4000 4000 4000>
    componentLinearAccScale=<0.4 0.4 0.4>
    componentLinearVelScale=<0.4 0.4 0.4>
    gravityOverride {z=-980}
    physicsBodyDefinitions [{
      boundBone {boneName=dyn_hairStrand_4}
      boxExtents=<20 30 30>
      constraintSetup {angularLimitsMax=<0 12 12>, angularLimitsMin=<0 -12 -12>, angularTarget=<1 0 0>}
      localJointOffset=<2 0 0>
     }]
   }
   n6 = Anim Dynamics - Bone: dyn_hairStrand_5!(ComponentPose=@n5, Alpha=0.75) -> (Pose) {
    alpha=0.75
    angularDampingOverride=1
    angularSpringConstant=12000
    bAngularSpring=true
    bOverrideAngularDamping=true
    gravityScale=0
    simulationSpace=RootRelative
    boundBone {boneName=dyn_hairStrand_5}
    componentAppliedLinearAccClamp=<4000 4000 4000>
    componentLinearAccScale=<0.4 0.4 0.4>
    componentLinearVelScale=<0.4 0.4 0.4>
    gravityOverride {z=-980}
    physicsBodyDefinitions [{
      boundBone {boneName=dyn_hairStrand_5}
      boxExtents=<20 30 30>
      constraintSetup {angularLimitsMax=<0 12 12>, angularLimitsMin=<0 -12 -12>, angularTarget=<1 0 0>}
      localJointOffset=<2 0 0>
     }]
   }
   Output Pose(Result=Component To Local(ComponentPose=@n6)) {layerGroup=None}
  ]}
 EventGraph {nodes [
   Event BlueprintUpdateAnimation()
   n1 = GetOwnerSkeletalMeshComponent()
   n2 = BreakRotator(InRot=GetDeltaTransformFromRefPose!(self=@n1, BoneName=head, BaseName=neck_02) -> (ReturnValue_Location:Vector, ReturnValue_Rotation:Rotator, ReturnValue_Scale:Vector).ReturnValue_Rotation)
   n3 = float * float(
    A=@n2.Yaw
    B=GetAlphaFromReferenceTransformDelta(SkeletalMeshComponent=@n1, BoneName=head, BaseName=neck_02, InputRange=(X=0,Y=-20), OutputRange=(X=0.2,Y=0.5), TwistAxis="1,0,0")
   ) {is=title K2Node_PromotableOperator}
   n4 = float * float(A=@n2, B=0.6)
   n5 = float * float(A=@n2.Pitch, B=0.5)
   > Set headRot(headRot:Rotator=Invert Rotator!(A:Rotator=MakeRotator(Roll=@n4, Pitch=@n5, Yaw=@n3)) -> (ReturnValue:Rotator)) -> (Output_Get:Rotator)
  ]}
]
variables [headRot {type=Rotator}]
